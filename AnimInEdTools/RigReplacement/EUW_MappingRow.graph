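# Editor utility widget row for the rig replacement tool that maps one source rig control to a target control picked from a searchable combo box.
# Selection broadcasts a dispatcher and reloads target channels from the parent mappings widget's pipe-delimited map.
# "Map Attributes" expands EUW_ChannelRow children, skipping the first source channel, whose changes feed the parent's extra_channels map.
# RefreshChannelRows restores saved channel picks; Set Row Color alternates shading and can flag unmapped rows.
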EditorUtilityWidgetBlueprint /AnimInEdTools/RigReplacement/EUW_MappingRow
graphs [
 EventGraph {nodes [
   n1 = "On Check State Changed (ExpandChannels_CheckBox)"!() -> (OutputDelegate, then, bIsChecked:bool) {is=title K2Node_ComponentBoundEvent}
   > n2 = Branch(Condition=@n1.bIsChecked)
   n3 = Get ChannelContainer() -> (ChannelContainer:Vertical Box Object Reference)
   > Set Visibility(Visibility=Visible, self=@n3)
   > ClearChildren(self=Get ChannelContainer())
   > n4 = For Each Loop(Array=Get SourceChannels() -> (SourceChannels:Array of Strings))
   > Branch(Condition="Not Equal (Integer)"(A=@n4.Array Index))
   > n5 = Create Widget!(
    execute:Exec
    WorldContextObject:Object Reference
    Class:User Widget Class Reference=&/EUW_ChannelRow.EUW_ChannelRow_C
    OwningPlayer:Player Controller Object Reference
   ) -> (then:Exec, ReturnValue:EUW Channel Row Object Reference)
   > Initialize!(
    execute
    self:EUW Channel Row Object Reference=@n5
    SourceChannel:String=@n4
    TargetChannels:Array of Strings=Get TargetChannels() -> (TargetChannels:Array of Strings)
   ) -> (then)
   > n6 = AddChild(self=Get ChannelContainer(), Content=@n5)
   n7 = CustomEvent!() -> (OutputDelegate, then, SourceChannel:String, TargetChannel:String) {is=title K2Node_CustomEvent}
   Bind Event to On Channel Mapping Changed(execute=@n6.then, self:EUW Channel Row Object Reference=@n5, Delegate=@n7) {is=title K2Node_AddDelegate}
   Set Visibility(execute=@n2.else, Visibility=Collapsed, self=@n3)
   > ClearChildren(self=Get ChannelContainer())
   n8 = "On Selection Changed (TargetControlComboBox)"!() -> (OutputDelegate, then, SelectedItem:String, SelectionType:ESelectInfo Enum) {is=title K2Node_ComponentBoundEvent}
   > Call On Selection Changed Dispatcher!(
    execute:Exec
    self:EUW Mapping Row Object Reference
    MappingRow:EUW Mapping Row Object Reference=Self-Reference()
    Selected Control:String=@n8.SelectedItem
   ) -> (then:Exec) {is=title K2Node_CallDelegate}
   n9 = Get TargetChannelMap!(self:EUW Mappings Widget Object Reference=Get ParentMappingsWidget() -> (ParentMappingsWidget:EUW Mappings Widget Object Reference)) -> (TargetChannelMap:Map of Strings to Strings)
   > Set TargetChannels(TargetChannels:Array of Strings=ParseIntoArray(SourceString=Find(TargetMap=@n9, Key=@n8.SelectedItem), Delimiter=|)) -> (Output_Get:Array of Strings)
   > Branch(Condition=IsChecked(self=Get ExpandChannels_CheckBox() -> (ExpandChannels_CheckBox:Editor Utility Check Box Object Reference)))
   > ClearChildren(self=Get ChannelContainer())
   > n10 = For Each Loop(Array=Get SourceChannels())
   > Branch(Condition="Not Equal (Integer)"(A=@n10.Array Index))
   > n11 = Create Widget(Class=&EUW_ChannelRow_C)
   > Initialize(self=@n11, SourceChannel=@n10, TargetChannels=Get TargetChannels())
   > n12 = AddChild(self=Get ChannelContainer(), Content=@n11)
   n13 = CustomEvent_1!() -> (OutputDelegate, then, SourceChannel:String, TargetChannel:String) {is=title K2Node_CustomEvent}
   Bind Event to On Channel Mapping Changed(execute=@n12.then, self=@n11, Delegate=@n13)
   Event PreConstruct()
   > Set Text(
    Text="To Text (String)"(InString=Get SourceControlName() -> (SourceControlName:String))
    self=Get SourceRigControlText() -> (SourceRigControlText:Text Object Reference)
   )
   Add!(
    execute=@n13.then
    TargetMap:Map of Strings to Strings=Get extra_channels!(self:EUW Mappings Widget Object Reference=Get ParentMappingsWidget()) -> (extra_channels:Map of Strings to Strings)
    Key:"String (by ref)"=@n13.SourceChannel
    Value:"String (by ref)"=@n13.TargetChannel
   ) -> (then) {is=title K2Node_CallFunction}
   Add(execute=@n7.then, TargetMap=Get extra_channels(self=Get ParentMappingsWidget()), Key=@n7.SourceChannel, Value=@n7.TargetChannel)
  ]}
 OnSelectionChanged_Dispatcher {nodes [OnSelectionChanged_Dispatcher!() -> (then, MappingRow:EUW Mapping Row Object Reference, Selected Control:String) {is=title K2Node_FunctionEntry}]}
 On_TargetControlComboBox_InitialExpand {nodes [
   On_TargetControlComboBox_InitialExpand!() -> (then, Item:String) {is=title K2Node_FunctionEntry}
   > Set Row Color!(execute, self, Sorted Row Index:int=Get Sorted Row Index() -> (Sorted Row Index:int=0), Ignore None Row:bool=false) -> (then) {is=title K2Node_CallFunction}
   > Return Node!(execute, ReturnValue:Widget Object Reference)
  ]}
 RefreshChannelRows {nodes [
   n1 = RefreshChannelRows!() -> (then, TargetCtrl:String, bExpanded:bool=~) {is=title K2Node_FunctionEntry}
   > SetIsChecked!(execute, self:Check Box Object Reference=Get ExpandChannels_CheckBox(), InIsChecked:bool=@n1.bExpanded) -> (then)
   n2 = Get TargetCtrl() -> (TargetCtrl:String)
   n3 = Get TargetChannelMap(self=Get ParentMappingsWidget())
   > Set TargetChannels(TargetChannels=ParseIntoArray(SourceString=Find(TargetMap=@n3, Key=@n2), Delimiter=|))
   > n4 = Branch(Condition=Get bExpanded() -> (bExpanded:bool))
   > Set Visibility(Visibility=Visible, self=Get ChannelContainer())
   > ClearChildren(self=Get ChannelContainer())
   > n5 = For Each Loop(Array=Get SourceChannels())
   > Branch(Condition="Not Equal (Integer)"(A=@n5.Array Index))
   > n6 = Create Widget(Class=&EUW_ChannelRow_C)
   > Initialize(self=@n6, SourceChannel=@n5, TargetChannels=Get TargetChannels())
   > AddChild(self=Get ChannelContainer(), Content=@n6)
   > Bind Event to On Channel Mapping Changed(self=@n6, Delegate=Create Event())
   n7 = Find(TargetMap=Get extra_channels(self=Get ParentMappingsWidget()), Key=@n5)
   > Branch(Condition="Not Equal Exactly (String)"(A=@n7))
   > SetSelectedOption!(
    execute
    self:"Searchable Combo Box (String) Object Reference"=Get TargetChannelCombBox!(self:EUW Channel Row Object Reference=@n6) -> (TargetChannelCombBox:"Searchable Combo Box (String) Object Reference")
    Option:String=@n7
   ) -> (then)
   Set Visibility(execute=@n4.else, Visibility=Collapsed, self=Get ChannelContainer())
   > ClearChildren(self=Get ChannelContainer())
  ]}
 Set Row Color {nodes [
   n1 = Set Row Color!() -> (then, Sorted Row Index:int, Ignore None Row:bool=~) {is=title K2Node_FunctionEntry}
   > n2 = Branch(Condition=@n1.Ignore None Row)
   n3 = Branch(execute=[@n7.else, @n2.then], Condition="Equal (Integer)"(A="% (Integer)"(A=@n1, B=2) -> (ReturnValue=0)))
   > n4 = Set Current Row Color(Current Row Color:Linear Color Structure=Get Assigned Row Color 0() -> (Assigned Row Color 0:Linear Color Structure)) -> (Output_Get:Linear Color Structure)
   Set BrushColor!(
    execute=[@n4.then, @n5.then]
    BrushColor:Linear Color Structure=Get Current Row Color() -> (Current Row Color:Linear Color Structure)
    self:Border Object Reference=Get RowBackground() -> (RowBackground:Border Object Reference)
   ) -> (then, Output_Get:Linear Color Structure)
   n5 = Set Current Row Color(execute=@n3.else, Current Row Color=Get Assigned Row Color 1() -> (Assigned Row Color 1:Linear Color Structure))
   n6 = Get TargetControlComboBox() -> (TargetControlComboBox:"Searchable Combo Box (String) Object Reference")
   n7 = Branch(
    execute=@n2.else
    Condition="Equal (Integer)"(A=GetSelectedIndex!(self:"Searchable Combo Box (String) Object Reference"=@n6) -> (ReturnValue:int=0))
   )
   > Set BrushColor(BrushColor=Get Unmapped Row Color() -> (Unmapped Row Color:Linear Color Structure), self=Get RowBackground())
  ]}
]
variables [
 Assigned Row Color 0 {type=LinearColor, default=<rgba 0.035 0.035 0.035 1>}
 Assigned Row Color 1 {type=LinearColor, default=<rgba 0.0226 0.0226 0.0226 1>}
 Current Row Color {type=LinearColor, default=<rgba 0.0495 0.0564 0.151 1>}
 Duplicate Row Color {type=LinearColor, default=<rgba 0.1059 0.0864 0.0355 1>}
 OnSelectionChanged_Dispatcher {default=()}
 ParentMappingsWidget {type=EUW_MappingsWidget_C, default=null}
 Sorted Row Index {type=integer}
 SourceChannels {type="array<string>", default []}
 SourceControlName {type=string, default=""}
 TargetChannels {type="array<string>", default []}
 Unmapped Row Color {type=LinearColor, default=<rgba 0.1059 0.0258 0.0241 1>}
]
widgets {
 info {rootWidgetClass=&Border, widgetCount=9}
 widgets [
  RowBackground {
   widgetClassPath=&Border
   bIsVariable=true
   widget=RowBackground
   properties {verticalAlignment=VAlign_Center, brushColor=<rgba 0.1927 0.1927 0.1927 1>, padding=<ltrb 19 2 4 2>}
  }
  VerticalBox_0 {
   widgetClassPath=&VerticalBox
   parent=RowBackground
   widget=VerticalBox_0
   slot {class=BorderSlot, properties {verticalAlignment=VAlign_Center, padding=<ltrb 19 2 4 2>}}
  }
  HorizontalBox_0 {
   widgetClassPath=&HorizontalBox
   parent=VerticalBox_0
   widget=HorizontalBox_0
   properties {bOverride_Cursor=true}
   slot {class=VerticalBoxSlot}
  }
  ExpandChannels_CheckBox {
   widgetClassPath=&EditorUtilityCheckBox
   bIsVariable=true
   parent=HorizontalBox_0
   widget=ExpandChannels_CheckBox
   properties {toolTipText="Map attributes that are not named the same between old and new rigs. \r\nExample : IK Solve is now IKFK Switch"}
   slot {class=HorizontalBoxSlot, properties {horizontalAlignment=HAlign_Left, padding=<ltrb 20 0 0 0>}}
  }
  TextBlock_50 {
   widgetClassPath=&TextBlock
   parent=ExpandChannels_CheckBox
   widget=TextBlock_50
   properties {
    text=Map Attributes
    font {
     bForceMonospaced=false
     bMaterialIsStencil=false
     fontMaterial=null
     fontObject=&Roboto
     letterSpacing=0
     monospacedWidth=1
     size=10.5
     skewAmount=0
     typefaceFontName=Light
     outlineSettings {
      bApplyOutlineToDropShadows=false
      bMiteredCorners=false
      bSeparateFillAlpha=false
      outlineMaterial=null
      outlineSize=0
      outlineColor=<rgba 0 0 0 1>
     }
    }
   }
   slot {class=PanelSlot}
  }
  SourceRigControlText {
   widgetClassPath=&TextBlock
   bIsVariable=true
   parent=HorizontalBox_0
   widget=SourceRigControlText
   properties {
    text=Text Block
    font {
     bForceMonospaced=false
     bMaterialIsStencil=false
     fontMaterial=null
     fontObject=&Roboto
     letterSpacing=0
     monospacedWidth=1
     size=10.5
     skewAmount=0
     typefaceFontName=Regular
     outlineSettings {
      bApplyOutlineToDropShadows=false
      bMiteredCorners=false
      bSeparateFillAlpha=false
      outlineMaterial=null
      outlineSize=0
      outlineColor=<rgba 0 0 0 1>
     }
    }
   }
   slot {
    class=HorizontalBoxSlot
    properties {horizontalAlignment=HAlign_Right, verticalAlignment=VAlign_Center, padding=<ltrb 0 0 10 0>, size {sizeRule=Fill, value=1}}
   }
  }
  TextBlock_263 {
   widgetClassPath=&TextBlock
   parent=HorizontalBox_0
   widget=TextBlock_263
   properties {
    text="->"
    font {
     bForceMonospaced=false
     bMaterialIsStencil=false
     fontMaterial=null
     fontObject=&Roboto
     letterSpacing=0
     monospacedWidth=1
     size=15
     skewAmount=0
     typefaceFontName=Light
     outlineSettings {
      bApplyOutlineToDropShadows=false
      bMiteredCorners=false
      bSeparateFillAlpha=false
      outlineMaterial=null
      outlineSize=0
      outlineColor=<rgba 0 0 0 1>
     }
    }
   }
   slot {class=HorizontalBoxSlot, properties {verticalAlignment=VAlign_Center}}
  }
  TargetControlComboBox {
   widgetClassPath=/Script/FortniteEditor.SearchableStringComboBox
   bIsVariable=true
   parent=HorizontalBox_0
   widget=TargetControlComboBox
   properties {
    selectedOption="<None> "
    defaultOptions ["<None> "]
    font {
     bForceMonospaced=false
     bMaterialIsStencil=false
     fontMaterial=null
     fontObject=&Roboto
     letterSpacing=0
     monospacedWidth=1
     size=10.5
     skewAmount=0
     typefaceFontName=Regular
     outlineSettings {
      bApplyOutlineToDropShadows=false
      bMiteredCorners=false
      bSeparateFillAlpha=false
      outlineMaterial=null
      outlineSize=0
      outlineColor=<rgba 0 0 0 1>
     }
    }
    foregroundColor {colorUseRule=UseColor_Specified, specifiedColor=<rgba 0.5271 0.5271 0.5271 1>}
    itemStyle {
     bUseParentRowBrush=false
     activeBrush {
      bIsDynamicallyLoaded=false
      drawAs=Image
      imageType=NoImage
      mirroring=NoMirror
      resourceName=None
      resourceObject=null
      tiling=NoTile
      imageSize=<0 0>
      margin=<ltrb 0 0 0 0>
      outlineSettings {
       bUseBrushTransparency=false
       roundingType=HalfHeightRadius
       width=0
       color {colorUseRule=UseColor_Specified, specifiedColor=<rgba 0 0 0 0>}
       cornerRadii=<0 0 0 0>
      }
      tintColor {colorUseRule=UseColor_Specified, specifiedColor=<rgba 0 0.162 0.7454 1>}
      uVRegion {bIsValid=false, max=<0 0>, min=<0 0>}
     }
     activeHighlightedBrush {
      bIsDynamicallyLoaded=false
      drawAs=Image
      imageType=NoImage
      mirroring=NoMirror
      resourceName=None
      resourceObject=null
      tiling=NoTile
      imageSize=<0 0>
      margin=<ltrb 0 0 0 0>
      outlineSettings {
       bUseBrushTransparency=false
       roundingType=HalfHeightRadius
       width=0
       color {colorUseRule=UseColor_Specified, specifiedColor=<rgba 0 0 0 0>}
       cornerRadii=<0 0 0 0>
      }
      tintColor {colorUseRule=UseColor_Specified, specifiedColor=<rgba 0.0252 0.0319 0.0423 1>}
      uVRegion {bIsValid=false, max=<0 0>, min=<0 0>}
     }
     activeHoveredBrush {
      bIsDynamicallyLoaded=false
      drawAs=Image
      imageType=NoImage
      mirroring=NoMirror
      resourceName=None
      resourceObject=null
      tiling=NoTile
      imageSize=<0 0>
      margin=<ltrb 0 0 0 0>
      outlineSettings {
       bUseBrushTransparency=false
       roundingType=HalfHeightRadius
       width=0
       color {colorUseRule=UseColor_Specified, specifiedColor=<rgba 0 0 0 0>}
       cornerRadii=<0 0 0 0>
      }
      tintColor {colorUseRule=UseColor_Specified, specifiedColor=<rgba 0 0.162 0.7454 1>}
      uVRegion {bIsValid=false, max=<0 0>, min=<0 0>}
     }
     dropIndicator_Above {
      bIsDynamicallyLoaded=false
      drawAs=Box
      imageType=FullColor
      mirroring=NoMirror
      resourceName=../../../Engine/Content/Slate/Common/DropZoneIndicator_Above.png
      resourceObject=null
      tiling=NoTile
      imageSize=<0 0>
      margin=<ltrb 0.625 0.625 0 0>
      outlineSettings {
       bUseBrushTransparency=false
       roundingType=HalfHeightRadius
       width=0
       color {colorUseRule=UseColor_Specified, specifiedColor=<rgba 0 0 0 0>}
       cornerRadii=<0 0 0 0>
      }
      tintColor {colorUseRule=UseColor_Specified, specifiedColor=<rgba 0.0091 0.3515 0.9647 1>}
      uVRegion {bIsValid=false, max=<0 0>, min=<0 0>}
     }
     dropIndicator_Below {
      bIsDynamicallyLoaded=false
      drawAs=Box
      imageType=FullColor
      mirroring=NoMirror
      resourceName=../../../Engine/Content/Slate/Common/DropZoneIndicator_Below.png
      resourceObject=null
      tiling=NoTile
      imageSize=<0 0>
      margin=<ltrb 0.625 0 0 0.625>
      outlineSettings {
       bUseBrushTransparency=false
       roundingType=HalfHeightRadius
       width=0
       color {colorUseRule=UseColor_Specified, specifiedColor=<rgba 0 0 0 0>}
       cornerRadii=<0 0 0 0>
      }
      tintColor {colorUseRule=UseColor_Specified, specifiedColor=<rgba 0.0091 0.3515 0.9647 1>}
      uVRegion {bIsValid=false, max=<0 0>, min=<0 0>}
     }
     dropIndicator_Onto {
      bIsDynamicallyLoaded=false
      drawAs=Box
      imageType=FullColor
      mirroring=NoMirror
      resourceName=../../../Engine/Content/Slate/Common/DropZoneIndicator_Onto.png
      resourceObject=null
      tiling=NoTile
      imageSize=<0 0>
      margin=<ltrb 0.25 0.25 0.25 0.25>
      outlineSettings {
       bUseBrushTransparency=false
       roundingType=HalfHeightRadius
       width=0
       color {colorUseRule=UseColor_Specified, specifiedColor=<rgba 0 0 0 0>}
       cornerRadii=<0 0 0 0>
      }
      tintColor {colorUseRule=UseColor_Specified, specifiedColor=<rgba 0.0091 0.3515 0.9647 1>}
      uVRegion {bIsValid=false, max=<0 0>, min=<0 0>}
     }
     evenRowBackgroundBrush {
      bIsDynamicallyLoaded=false
      drawAs=Image
      imageType=NoImage
      mirroring=NoMirror
      resourceName=None
      resourceObject=null
      tiling=NoTile
      imageSize=<0 0>
      margin=<ltrb 0 0 0 0>
      outlineSettings {
       bUseBrushTransparency=false
       roundingType=HalfHeightRadius
       width=0
       color {colorUseRule=UseColor_Specified, specifiedColor=<rgba 0 0 0 0>}
       cornerRadii=<0 0 0 0>
      }
      tintColor {colorUseRule=UseColor_Specified, specifiedColor=<rgba 0.0103 0.0103 0.0103 1>}
      uVRegion {bIsValid=false, max=<0 0>, min=<0 0>}
     }
     evenRowBackgroundHoveredBrush {
      bIsDynamicallyLoaded=false
      drawAs=Image
      imageType=NoImage
      mirroring=NoMirror
      resourceName=None
      resourceObject=null
      tiling=NoTile
      imageSize=<0 0>
      margin=<ltrb 0 0 0 0>
      outlineSettings {
       bUseBrushTransparency=false
       roundingType=HalfHeightRadius
       width=0
       color {colorUseRule=UseColor_Specified, specifiedColor=<rgba 0 0 0 0>}
       cornerRadii=<0 0 0 0>
      }
      tintColor {colorUseRule=UseColor_Specified, specifiedColor=<rgba 0.0176 0.0176 0.0176 1>}
      uVRegion {bIsValid=false, max=<0 0>, min=<0 0>}
     }
     inactiveBrush {
      bIsDynamicallyLoaded=false
      drawAs=Image
      imageType=NoImage
      mirroring=NoMirror
      resourceName=None
      resourceObject=null
      tiling=NoTile
      imageSize=<0 0>
      margin=<ltrb 0 0 0 0>
      outlineSettings {
       bUseBrushTransparency=false
       roundingType=HalfHeightRadius
       width=0
       color {colorUseRule=UseColor_Specified, specifiedColor=<rgba 0 0 0 0>}
       cornerRadii=<0 0 0 0>
      }
      tintColor {colorUseRule=UseColor_Specified, specifiedColor=<rgba 0.0513 0.0953 0.159 1>}
      uVRegion {bIsValid=false, max=<0 0>, min=<0 0>}
     }
     inactiveHighlightedBrush {
      bIsDynamicallyLoaded=false
      drawAs=Image
      imageType=NoImage
      mirroring=NoMirror
      resourceName=None
      resourceObject=null
      tiling=NoTile
      imageSize=<0 0>
      margin=<ltrb 0 0 0 0>
      outlineSettings {
       bUseBrushTransparency=false
       roundingType=HalfHeightRadius
       width=0
       color {colorUseRule=UseColor_Specified, specifiedColor=<rgba 0 0 0 0>}
       cornerRadii=<0 0 0 0>
      }
      tintColor {colorUseRule=UseColor_Specified, specifiedColor=<rgba 0.0252 0.0319 0.0423 1>}
      uVRegion {bIsValid=false, max=<0 0>, min=<0 0>}
     }
     inactiveHoveredBrush {
      bIsDynamicallyLoaded=false
      drawAs=Image
      imageType=NoImage
      mirroring=NoMirror
      resourceName=None
      resourceObject=null
      tiling=NoTile
      imageSize=<0 0>
      margin=<ltrb 0 0 0 0>
      outlineSettings {
       bUseBrushTransparency=false
       roundingType=HalfHeightRadius
       width=0
       color {colorUseRule=UseColor_Specified, specifiedColor=<rgba 0 0 0 0>}
       cornerRadii=<0 0 0 0>
      }
      tintColor {colorUseRule=UseColor_Specified, specifiedColor=<rgba 0.0513 0.0953 0.159 1>}
      uVRegion {bIsValid=false, max=<0 0>, min=<0 0>}
     }
     oddRowBackgroundBrush {
      bIsDynamicallyLoaded=false
      drawAs=Image
      imageType=NoImage
      mirroring=NoMirror
      resourceName=None
      resourceObject=null
      tiling=NoTile
      imageSize=<0 0>
      margin=<ltrb 0 0 0 0>
      outlineSettings {
       bUseBrushTransparency=false
       roundingType=HalfHeightRadius
       width=0
       color {colorUseRule=UseColor_Specified, specifiedColor=<rgba 0 0 0 0>}
       cornerRadii=<0 0 0 0>
      }
      tintColor {colorUseRule=UseColor_Specified, specifiedColor=<rgba 0.0103 0.0103 0.0103 1>}
      uVRegion {bIsValid=false, max=<0 0>, min=<0 0>}
     }
     oddRowBackgroundHoveredBrush {
      bIsDynamicallyLoaded=false
      drawAs=Image
      imageType=NoImage
      mirroring=NoMirror
      resourceName=None
      resourceObject=null
      tiling=NoTile
      imageSize=<0 0>
      margin=<ltrb 0 0 0 0>
      outlineSettings {
       bUseBrushTransparency=false
       roundingType=HalfHeightRadius
       width=0
       color {colorUseRule=UseColor_Specified, specifiedColor=<rgba 0 0 0 0>}
       cornerRadii=<0 0 0 0>
      }
      tintColor {colorUseRule=UseColor_Specified, specifiedColor=<rgba 0.0176 0.0176 0.0176 1>}
      uVRegion {bIsValid=false, max=<0 0>, min=<0 0>}
     }
     parentRowBackgroundBrush {
      bIsDynamicallyLoaded=false
      drawAs=Image
      imageType=NoImage
      mirroring=NoMirror
      resourceName=None
      resourceObject=null
      tiling=NoTile
      imageSize=<32 32>
      margin=<ltrb 0 0 0 0>
      outlineSettings {
       bUseBrushTransparency=false
       roundingType=HalfHeightRadius
       width=0
       color {colorUseRule=UseColor_Specified, specifiedColor=<rgba 0 0 0 0>}
       cornerRadii=<0 0 0 0>
      }
      tintColor {colorUseRule=UseColor_Specified, specifiedColor=<rgba 1 1 1 1>}
      uVRegion {bIsValid=false, max=<0 0>, min=<0 0>}
     }
     parentRowBackgroundHoveredBrush {
      bIsDynamicallyLoaded=false
      drawAs=Image
      imageType=NoImage
      mirroring=NoMirror
      resourceName=None
      resourceObject=null
      tiling=NoTile
      imageSize=<32 32>
      margin=<ltrb 0 0 0 0>
      outlineSettings {
       bUseBrushTransparency=false
       roundingType=HalfHeightRadius
       width=0
       color {colorUseRule=UseColor_Specified, specifiedColor=<rgba 0 0 0 0>}
       cornerRadii=<0 0 0 0>
      }
      tintColor {colorUseRule=UseColor_Specified, specifiedColor=<rgba 1 1 1 1>}
      uVRegion {bIsValid=false, max=<0 0>, min=<0 0>}
     }
     selectedTextColor {colorUseRule=UseColor_Specified, specifiedColor=<rgba 1 1 1 1>}
     selectorFocusedBrush {
      bIsDynamicallyLoaded=false
      drawAs=Border
      imageType=FullColor
      mirroring=NoMirror
      resourceName=../../../Engine/Content/Slate/Common/Selector.png
      resourceObject=null
      tiling=Both
      imageSize=<0 0>
      margin=<ltrb 0.25 0.25 0.25 0.25>
      outlineSettings {
       bUseBrushTransparency=false
       roundingType=HalfHeightRadius
       width=0
       color {colorUseRule=UseColor_Specified, specifiedColor=<rgba 0 0 0 0>}
       cornerRadii=<0 0 0 0>
      }
      tintColor {colorUseRule=UseColor_Specified, specifiedColor=<rgba 0.701 0.225 0.003 1>}
      uVRegion {bIsValid=false, max=<0 0>, min=<0 0>}
     }
     textColor {colorUseRule=UseColor_Specified, specifiedColor=<rgba 0.5271 0.5271 0.5271 1>}
    }
    widgetStyle {
     comboButtonStyle {
      downArrowAlign=VAlign_Center
      buttonStyle {
       clickedSlateSound {resourceObject=null}
       disabled {
        bIsDynamicallyLoaded=false
        drawAs=RoundedBox
        imageType=NoImage
        mirroring=NoMirror
        resourceName=None
        resourceObject=null
        tiling=NoTile
        imageSize=<0 0>
        margin=<ltrb 0 0 0 0>
        outlineSettings {
         bUseBrushTransparency=false
         roundingType=FixedRadius
         width=1
         color {colorUseRule=UseColor_ColorTable, specifiedColor=<rgba 0 0 0 0>}
         cornerRadii=<4 4 4 4>
        }
        tintColor {colorUseRule=UseColor_Specified, specifiedColor=<rgba 0.0395 0.0395 0.0395 1>}
        uVRegion {bIsValid=false, max=<0 0>, min=<0 0>}
       }
       disabledForeground {colorUseRule=UseColor_ColorTable, specifiedColor=<rgba 0 0 0 0>}
       hovered {
        bIsDynamicallyLoaded=false
        drawAs=RoundedBox
        imageType=NoImage
        mirroring=NoMirror
        resourceName=None
        resourceObject=null
        tiling=NoTile
        imageSize=<0 0>
        margin=<ltrb 0 0 0 0>
        outlineSettings {
         bUseBrushTransparency=false
         roundingType=FixedRadius
         width=1
         color {colorUseRule=UseColor_ColorTable, specifiedColor=<rgba 0 0 0 0>}
         cornerRadii=<4 4 4 4>
        }
        tintColor {colorUseRule=UseColor_Specified, specifiedColor=<rgba 0.0048 0.0048 0.0048 1>}
        uVRegion {bIsValid=false, max=<0 0>, min=<0 0>}
       }
       hoveredForeground {colorUseRule=UseColor_ColorTable, specifiedColor=<rgba 0 0 0 0>}
       hoveredSlateSound {resourceObject=null}
       normal {
        bIsDynamicallyLoaded=false
        drawAs=RoundedBox
        imageType=NoImage
        mirroring=NoMirror
        resourceName=None
        resourceObject=null
        tiling=NoTile
        imageSize=<0 0>
        margin=<ltrb 0 0 0 0>
        outlineSettings {
         bUseBrushTransparency=false
         roundingType=FixedRadius
         width=1
         color {colorUseRule=UseColor_ColorTable, specifiedColor=<rgba 0 0 0 0>}
         cornerRadii=<4 4 4 4>
        }
        tintColor {colorUseRule=UseColor_Specified, specifiedColor=<rgba 0.0048 0.0048 0.0048 1>}
        uVRegion {bIsValid=false, max=<0 0>, min=<0 0>}
       }
       normalForeground {colorUseRule=UseColor_ColorTable, specifiedColor=<rgba 0 0 0 0>}
       normalPadding=<ltrb 8 1 8 1>
       pressed {
        bIsDynamicallyLoaded=false
        drawAs=RoundedBox
        imageType=NoImage
        mirroring=NoMirror
        resourceName=None
        resourceObject=null
        tiling=NoTile
        imageSize=<0 0>
        margin=<ltrb 0 0 0 0>
        outlineSettings {
         bUseBrushTransparency=false
         roundingType=FixedRadius
         width=1
         color {colorUseRule=UseColor_ColorTable, specifiedColor=<rgba 0 0 0 0>}
         cornerRadii=<4 4 4 4>
        }
        tintColor {colorUseRule=UseColor_Specified, specifiedColor=<rgba 0.0048 0.0048 0.0048 1>}
        uVRegion {bIsValid=false, max=<0 0>, min=<0 0>}
       }
       pressedForeground {colorUseRule=UseColor_ColorTable, specifiedColor=<rgba 0 0 0 0>}
       pressedPadding=<ltrb 8 2 8 0>
       pressedSlateSound {resourceObject=null}
      }
      contentPadding=<ltrb 0 0 0 0>
      downArrowImage {
       bIsDynamicallyLoaded=false
       drawAs=Image
       imageType=Vector
       mirroring=NoMirror
       resourceName=../../../Engine/Content/Slate/Starship/CoreWidgets/ComboBox/wide-chevron-down.svg
       resourceObject=null
       tiling=NoTile
       imageSize=<16 16>
       margin=<ltrb 0 0 0 0>
       outlineSettings {
        bUseBrushTransparency=false
        roundingType=HalfHeightRadius
        width=0
        color {colorUseRule=UseColor_Specified, specifiedColor=<rgba 0 0 0 0>}
        cornerRadii=<0 0 0 0>
       }
       tintColor {colorUseRule=UseColor_Specified, specifiedColor=<rgba 1 1 1 1>}
       uVRegion {bIsValid=false, max=<0 0>, min=<0 0>}
      }
      downArrowPadding=<ltrb 2 2 2 2>
      menuBorderBrush {
       bIsDynamicallyLoaded=false
       drawAs=RoundedBox
       imageType=NoImage
       mirroring=NoMirror
       resourceName=None
       resourceObject=null
       tiling=NoTile
       imageSize=<0 0>
       margin=<ltrb 0 0 0 0>
       outlineSettings {
        bUseBrushTransparency=false
        roundingType=FixedRadius
        width=1
        color {colorUseRule=UseColor_ColorTable, specifiedColor=<rgba 0 0 0 0>}
        cornerRadii=<0 0 0 0>
       }
       tintColor {colorUseRule=UseColor_Specified, specifiedColor=<rgba 0.0395 0.0395 0.0395 1>}
       uVRegion {bIsValid=false, max=<0 0>, min=<0 0>}
      }
      menuBorderPadding=<ltrb 0 0 0 0>
      shadowColorAndOpacity=<rgba 0 0 0 1>
      shadowOffset=<0 0>
     }
     contentPadding=<ltrb 0 0 0 0>
     menuRowPadding=<ltrb 8 3 8 3>
     pressedSlateSound {resourceObject=null}
     selectionChangeSlateSound {resourceObject=null}
    }
   }
   slot {class=HorizontalBoxSlot, properties {verticalAlignment=VAlign_Center, padding=<ltrb 10 0 15 0>, size {sizeRule=Fill, value=1}}}
  }
  ChannelContainer {
   widgetClassPath=&VerticalBox
   bIsVariable=true
   parent=VerticalBox_0
   widget=ChannelContainer
   properties {visibility=Collapsed}
   slot {class=VerticalBoxSlot}
  }
 ]
}
