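# Editor utility widget for mapping controls and animation channels from a source Control Rig to a target rig, which can come from an Animatable Creature.
# Embedded Python reads both rig hierarchies to fill per-control rows, then a library DataTable or matching names auto-suggest targets.
# Helpers mirror left/right names and flag unmapped or duplicate rows for filtering.
# Python creates the output mapping DataTable beside the target creature's rig.
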EditorUtilityWidgetBlueprint /AnimInEdTools/RigReplacement/EUW_MappingsWidget
graphs [
 Auto Suggest Mapping of Controls {nodes [
   Auto Suggest Mapping of Controls() {is=title K2Node_FunctionEntry}
   > n1 = Sequence!(execute) -> (then_0, then_1, then_2:Exec)
   > Clear!(execute, TargetMap:Map of Strings to Strings=Get extra_channels() -> (extra_channels:Map of Strings to Strings)) -> (then) {is=title K2Node_CallFunction}
   > Clear(TargetMap=Get LibraryLookup() -> (LibraryLookup:Map of Strings to Strings))
   > Clear(TargetMap=Get LibraryLookup_Specific() -> (LibraryLookup_Specific:Map of Strings to Strings))
   n2 = GetDataTableRowNames(execute=@n1.then_1, Table=Get MappingLibrary() -> (MappingLibrary:Data Table Object Reference))
   > n3 = For Each Loop(Array:Array of Names=@n2) -> (Array Element:Name)
   > n4 = Get Data Table Row!(execute:Exec, DataTable:Data Table Object Reference=Get MappingLibrary(), RowName:Name=@n3) -> (then:Exec, RowNotFound:Exec, ReturnValue:Mapping Library Struct Structure)
   n5 = Break Mapping Library Struct!(MappingLibraryStruct:"Mapping Library Struct Structure (by ref)"=@n4) -> (
    Source_Module_27_#1:String
    Target_Module_30_#2:String
    Source_CRM_Class_21_#3:String
    Target_CRM_Class_25_#4:String
    SourceCtrl_2_#5:String
    TargetCtrl_4_#6:String
    MapAttribute_15_#7:bool=~
    SourceChannel_17_#8:String
    TargetChannel_19_#9:String
    Source_Module_Deprecated_26_#10:bool=~
   ) {is=title K2Node_BreakStruct}
   > Set SourceCtrl(SourceCtrl:String=Append!(A=@n5, B=/, C:String=@n5.SourceCtrl_2_#5) -> (ReturnValue)) -> (Output_Get:String)
   > Set SourceChannel(SourceChannel:String=@n5.SourceChannel_17_#8) -> (Output_Get:String)
   > n6 = Set Target_Module(Target_Module:String=@n5.Target_Module_30_#2) -> (Output_Get:String)
   > n7 = Set TargetCtrl(TargetCtrl:String=Append(A=@n6, B=/, C=@n5.TargetCtrl_4_#6)) -> (Output_Get:String)
   > Branch(Condition=Contains Item!(TargetSet:Set of Strings=Get TargetControls() -> (TargetControls:Set of Strings), ItemToFind:"String (by ref)"=@n7) -> (ReturnValue:bool) {is=title K2Node_CallFunction})
   > n8 = Branch(Condition="Equal, Case Insensitive (String)"(A=@n5.Target_CRM_Class_25_#4))
   Add!(
    execute=[@n10.then, @n8.then, @n9.else]
    TargetMap:Map of Strings to Strings=Get LibraryLookup()
    Key:"String (by ref)"=Get SourceCtrl() -> (SourceCtrl:String)
    Value:"String (by ref)"=Get TargetCtrl() -> (TargetCtrl:String)
   ) -> (then) {is=title K2Node_CallFunction}
   > Branch(Condition=@n5.MapAttribute_15_#7)
   > Add(
    TargetMap=Get LibrarySrcChan() -> (LibrarySrcChan:Map of Strings to Strings)
    Key=Get SourceCtrl()
    Value=Get SourceChannel() -> (SourceChannel:String)
   )
   > Add(TargetMap=Get extra_channels(), Key=Get SourceChannel(), Value=@n5.TargetChannel_19_#9)
   n9 = Branch(
    execute=@n8.else
    Condition="Not Equal, Case Insenstive (String)"!(A:String=@n5.Target_CRM_Class_25_#4, B:String) -> (ReturnValue:bool)
   )
   > n10 = Add(TargetMap=Get LibraryLookup_Specific(), Key=Get SourceCtrl(), Value=Get TargetCtrl())
   n11 = "For Each Loop (Map)"!(
    execute:Exec=@n1.then_2
    MapPin:Map of Strings to EUW Mapping Row Object References=Get SourceControls() -> (SourceControls:Map of Strings to EUW Mapping Row Object References)
    BreakPin:Exec
   ) -> (then:Exec, KeyPin:String, ValuePin:EUW Mapping Row Object Reference, CompletedPin:Exec)
   > Set Source Key(Source Key:String=@n11) -> (Output_Get:String)
   n12 = Find(TargetMap=Get LibraryLookup_Specific(), Key=Get Source Key() -> (Source Key:String))
   > n13 = Branch(Condition=@n12.ReturnValue)
   > n14 = Branch(Condition=Contains Item(TargetSet=Get TargetControls(), ItemToFind=@n12))
   n15 = Get TargetControlComboBox!(self:EUW Mapping Row Object Reference=@n11.ValuePin) -> (TargetControlComboBox:"Searchable Combo Box (String) Object Reference")
   > SetSelectedOption!(execute, self:"Searchable Combo Box (String) Object Reference"=@n15, Option:String=@n12) -> (then)
   n16 = Find(TargetMap=Get LibraryLookup(), Key=Get Source Key())
   n17 = Branch(execute=[@n13.else, @n14.else], Condition=@n16.ReturnValue)
   > n18 = Branch(Condition=Contains Item(TargetSet=Get TargetControls(), ItemToFind=@n16))
   > SetSelectedOption(self=@n15, Option=@n16)
   > Branch(Condition=Find(TargetMap=Get LibrarySrcChan(), Key=Get Source Key()).ReturnValue)
   > RefreshChannelRows!(
    execute
    self:EUW Mapping Row Object Reference=Find(TargetMap:Map of Strings to EUW Mapping Row Object References=Get SourceControls(), Key=Get Source Key()) -> (Value:EUW Mapping Row Object Reference)
    TargetCtrl:String=@n16
    bExpanded:bool=true
   ) -> (then)
   n19 = Branch(execute=[@n17.else, @n18.else], Condition=Contains Item(TargetSet=Get TargetControls(), ItemToFind=Get Source Key()))
   n20 = Get TargetControlComboBox(self=@n11.ValuePin)
   > SetSelectedOption(self=@n20, Option=Get Source Key())
   > Branch(Condition=Find(TargetMap:Map of Strings to Strings=Get LibrarySrcChan(), Key=Get Source Key()) -> (Value:String).ReturnValue)
   > RefreshChannelRows(
    self=Find(TargetMap:Map of Strings to EUW Mapping Row Object References=Get SourceControls(), Key=Get Source Key()) -> (Value:EUW Mapping Row Object Reference)
    TargetCtrl=Get Source Key()
    bExpanded=true
   )
   SetSelectedIndex!(execute=@n19.else, self:"Searchable Combo Box (String) Object Reference"=@n20, Index:int=0) -> (then)
  ]}
 Clear Source Ctrls {nodes [
   Clear Source Ctrls() {is=title K2Node_FunctionEntry}
   > n1 = "For Each Loop (Map)"(MapPin=Get SourceControls())
   > RemoveFromParent(self=@n1.ValuePin)
  ]}
 Create Data Table {nodes [
   Create Data Table() {is=title K2Node_FunctionEntry}
   > n1 = LoadAsset_Blocking(Asset=Get Target_AC() -> (Target_AC:Animatable Creature Soft Object Reference))
   > n2 = Cast To AnimatableCreature!(execute, Object=@n1) -> (then, CastFailed, AsAnimatable Creature:Animatable Creature Object Reference) {is=title K2Node_DynamicCast}
   n3 = Get TargetRig() -> (TargetRig:Control Rig Blueprint Object Reference)
   > Set Data Table Name(Data Table Name:String=Append!(A=GetDisplayName(Object=@n3), B=_ControlMappings) -> (ReturnValue)) -> (Output_Get:String)
   n4 = Get Control Rig!(self:Animatable Creature Object Reference=@n2) -> (Control Rig:Control Rig Asset Soft Reference Structure)
   > Set Data Table Location(Data Table Location:String=Split(
     SourceString="Get Path Name (Control Rig Soft Reference)"!(InRef:"Control Rig Asset Soft Reference Structure (by ref)"=@n4) -> (ReturnValue:String) {is=title K2Node_CallFunction}
     InStr=/
     SearchDir=FromEnd
    )) -> (Output_Get:String)
   > n5 = DoesAssetExist(AssetPath=Append!(A=Get Data Table Location() -> (Data Table Location:String), B=/, C:String=Get Data Table Name() -> (Data Table Name:String)) -> (ReturnValue))
   > n6 = Branch(Condition=@n5)
   n7 = Get Data Table Name()
   n8 = Get Data Table Location()
   > n9 = Show Message Dialog(
    Title="To Text (String)"(InString=Data Table Error)
    Message="To Text (String)"(InString=Append!(A="The ", B=@n7, C=" at ", D:String=@n8, E:String=" already exists. Would you like to overwrite the existing one with a new one?") -> (ReturnValue))
    MessageType=YesNo
    DefaultValue=No
    MessageCategory=Warning
   )
   > n10 = Branch(Condition="Equal (Enum)"(A:EAppReturnType Enum=@n9, B:EAppReturnType Enum=No))
   n11 = Execute Python Script!(
    execute=[@n6.else, @n10.else]
    PythonScript="import unreal\r\n\r\nfull_path = f\"{save_path}/{datatable_name}\"\r\n\r\n##Handle overwrite: delete_asset (path-based) consolidates the package cleanly\r\nif unreal.EditorAssetLibrary.does_asset_exist(full_path):\r\n  if not unreal.EditorAssetLibrary.delete_asset(full_path):\r\n      unreal.log_error(f\"Failed to delete existing asset: {full_path}\")\r\n\r\nasset_tools = unreal.AssetToolsHelpers.get_asset_tools()\r\nrow_struct = unreal.load_object(None, row_struct_path)\r\nif not row_struct:\r\n  unreal.log_error(f\"Failed to load row struct at path: {row_struct_path}\")\r\n\r\nfactory = unreal.DataTableFactory()\r\nfactory.struct = row_struct\r\ndata_table_asset = asset_tools.create_asset(\r\n  asset_name=datatable_name,\r\n  package_path=save_path,\r\n  asset_class=unreal.DataTable,\r\n  factory=factory\r\n)\r\n\r\nif not data_table_asset:\r\n  unreal.log_error(f\"Failed to create data table asset at path: {full_path}\")\r\n\r\nunreal.EditorAssetLibrary.save_asset(full_path)\r\n"
    save_path:String=Get Data Table Location()
    datatable_name:String=Get Data Table Name()
    row_struct_path:String=Get Mapping_Struct_Path() -> (Mapping_Struct_Path:String)
   ) -> (then, ReturnValue, data_table_asset:Data Table Object Reference) {inputs [save_path, datatable_name, row_struct_path], outputs [data_table_asset]}
   > Set New Created Data Table Asset(New Created Data Table Asset:Data Table Object Reference=@n11.data_table_asset) -> (Output_Get:Data Table Object Reference)
   > n12 = Return String Mapping!(execute, self=Self-Reference()) -> (then, Final Mapping:Map of Strings to Strings)
   > n13 = "For Each Loop (Map)"(MapPin:Map of Strings to Strings=@n12) -> (ValuePin:String)
   > Set TempMapKey(TempMapKey:String=@n13) -> (Output_Get:String)
   > Set TempMapValue(TempMapValue:String=@n13.ValuePin) -> (Output_Get:String)
   n14 = Get TempMapKey() -> (TempMapKey:String)
   n15 = Get ExpandChannels_CheckBox!(self:EUW Mapping Row Object Reference=Find(TargetMap=Get SourceControls(), Key=Get TempMapKey())) -> (ExpandChannels_CheckBox:Editor Utility Check Box Object Reference)
   n16 = Get SourceRig() -> (SourceRig:Control Rig Blueprint Object Reference)
   n17 = Get TargetRig()
   > AddDataTableRow!(
    execute
    DataTable:Data Table Object Reference=Get New Created Data Table Asset() -> (New Created Data Table Asset:Data Table Object Reference)
    RowName:"Name (by ref)"=String To Name(InString=Append!(A="Row ", B=@n14) -> (ReturnValue))
    RowData:Str to Str Mapping Struct Structure=Make Str To Str Mapping Struct!(
     mapattributes_23_#11:bool=IsChecked(self=@n15)
     SourceCtrl_6_#12:String=Get TempMapKey()
     TargetCtrl_8_#13:String=Get TempMapValue() -> (TempMapValue:String)
     SourceRig_15_#14:String=GetDisplayName(Object=@n16)
     TargetRig_16_#15:String=GetDisplayName(Object=@n17)
     SourceChannel_19_#16:String
     TargetChannel_25_#17:String
    ) -> (Str_to_Str_Mapping_Struct:Str to Str Mapping Struct Structure) {is=title K2Node_MakeStruct}
   ) -> (then)
   n18 = Get SourceControls()
   n19 = Get TempMapKey()
   > n20 = For Each Loop(Array:Array of Strings=Get SourceChannels!(self:EUW Mapping Row Object Reference=Find(TargetMap=@n18, Key=@n19)) -> (SourceChannels:Array of Strings)) -> (Array Element:String)
   > n21 = Set TempSourceChannel(TempSourceChannel:String=@n20) -> (Output_Get:String)
   > Branch(Condition=Find(TargetMap:Map of Strings to Strings=Get extra_channels(), Key=Get TempSourceChannel() -> (TempSourceChannel:String)) -> (Value:String).ReturnValue)
   n22 = Get SourceRig()
   n23 = Get extra_channels()
   n24 = Get TempSourceChannel()
   n25 = Get TargetRig()
   > AddDataTableRow(
    DataTable=Get New Created Data Table Asset()
    RowName=String To Name(InString=Append(A=Chan_, B=@n21))
    RowData=Make Str To Str Mapping Struct(
     mapattributes_23_#11=False
     SourceCtrl_6_#12=Get TempMapKey()
     TargetCtrl_8_#13=Get TempMapValue()
     SourceRig_15_#14=GetDisplayName(Object=@n22)
     TargetRig_16_#15=GetDisplayName(Object=@n25)
     SourceChannel_19_#16=Get TempSourceChannel()
     TargetChannel_25_#17=Find(TargetMap=@n23, Key=@n24)
    )
   )
   SaveLoadedAsset(execute=@n13.CompletedPin, AssetToSave=Get New Created Data Table Asset(), bOnlyIfIsDirty=false)
  ]}
 Data Table updated {nodes [Data Table updated() {is=title K2Node_FunctionEntry}]}
 ED Target Rig Value Changed {nodes [ED Target Rig Value Changed!() -> (then, Asset not none:bool=~) {is=title K2Node_FunctionEntry}]}
 EUW_MappingsWidget_AutoGenFunc {nodes [
   EUW_MappingsWidget_AutoGenFunc!() -> (then, MappingRow:EUW Mapping Row Object Reference) {is=title K2Node_FunctionEntry}
   > Set Row Colors!(execute, self, Ignore None Rows:bool=false) -> (then) {is=title K2Node_CallFunction}
   > Update Statistics()
  ]}
 EventGraph {nodes [
   "On Property Changed (SourceRig_SinglePropertyView)"!() -> (OutputDelegate, then, PropertyName:Name=None) {is=title K2Node_ComponentBoundEvent}
   > Set Visibility(Visibility=Visible, self=Get Text_LoadingRig() -> (Text_LoadingRig:Text Object Reference))
   > n1 = Set Is Pre Construct Event(Is Pre Construct Event:bool=false) -> (Output_Get:bool)
   n2 = Sequence(execute=[@n13.then, @n1.then])
   > Clear Source Ctrls() {is=title K2Node_CallFunction}
   Populate Source Ctrls and Alt Row Colors(execute=@n2.then_1)
   n3 = Branch(execute=@n2.then_2, Condition=Get Is Pre Construct Event() -> (Is Pre Construct Event:bool))
   n4 = Set Visibility(execute=@n3.else, Visibility=Collapsed, self=Get Text_LoadingRig())
   Get!(execute:Exec=[@n8.then, @n4.then]) -> (then:Exec, else:Exec, TargetRig:Control Rig Blueprint Object Reference)
   > n5 = Sequence()
   > "Reset Target Ctrls Mapping, Populate Target Ctrl Set"()
   > Set Visibility(Visibility=Collapsed, self=Get Text_LoadingRig())
   Auto Suggest Mapping of Controls(execute=@n5.then_1) {is=title K2Node_CallFunction}
   n6 = Set Row Colors(execute=@n5.then_2, Ignore None Rows=false)
   Filter Rows and Update Statistics!(execute=[@n6.then, @n9.then, @n10.Is Not Valid], self, Ignore None Rows:bool=false) -> (then)
   "On Property Changed (TargetRig_SinglePropertyView)"!() -> (OutputDelegate, then, PropertyName:Name=None) {is=title K2Node_ComponentBoundEvent}
   > Set Visibility(Visibility=Visible, self=Get Text_LoadingRig())
   > n7 = Is Valid(InputObject=Get TargetRig())
   Set bIsEnabled!(
    execute=@n7.Is Not Valid
    bIsEnabled:bool=false
    self:Widget Object Reference=Get ImportDataTableFile() -> (ImportDataTableFile:Single Property View Object Reference)
   ) -> (then, Output_Get:bool)
   > Set Visibility(Visibility=Collapsed, self=Get Text_LoadingRig())
   n8 = Set bIsEnabled(execute=@n7.Is Valid, bIsEnabled=true, self=Get ImportDataTableFile())
   n9 = "On Check State Changed (OnlyUnresolved_Checkbox)"!() -> (OutputDelegate, then, bIsChecked:bool) {is=title K2Node_ComponentBoundEvent}
   "On Clicked (Refresh_Button)"() {is=title K2Node_ComponentBoundEvent}
   > n10 = Is Valid(InputObject=Get Data_Table() -> (Data_Table:Data Table Object Reference))
   > Update Mapping From Data Table()
   "On Property Changed (ImportDataTableFile)"!() -> (OutputDelegate, then, PropertyName:Name=None) {is=title K2Node_ComponentBoundEvent}
   > n11 = Is Valid(InputObject=Get Data_Table())
   Filter Rows and Update Statistics(execute=@n11.Is Not Valid, Ignore None Rows=false)
   Update Mapping From Data Table(execute=@n11.Is Valid)
   "On Clicked (Create_Mapping_Btn)"() {is=title K2Node_ComponentBoundEvent}
   > Create Data Table() {is=title K2Node_CallFunction}
   "On Property Changed (TargetACFile)"!() -> (OutputDelegate, then, PropertyName:Name=None) {is=title K2Node_ComponentBoundEvent}
   > Set Visibility(Visibility=Visible, self=Get Text_LoadingRig())
   > Populate Target Rig when AC is changed()
   > Target or Source Rig Updated()
   > Set Visibility(Visibility=Collapsed, self=Get Text_LoadingRig())
   Event PreConstruct()
   > Set Is Pre Construct Event(Is Pre Construct Event=true)
   > n12 = For Each Loop(Array:Array of Animatable Creature Object References=Get Animatable Creatures!(Project:BP Animation Browser Project Object Reference, bSearchSubClasses:bool=false, __WorldContext:Object Reference) -> (Animatable Creatures:Array of Animatable Creature Object References) {is=title K2Node_CallFunction}) -> (Array Element:Animatable Creature Object Reference)
   SetObject(
    execute=@n12.Completed
    self=[
     Get ImportDataTableFile()
     Get TargetACFile() -> (TargetACFile:Single Property View Object Reference)
     Get SourceRig_SinglePropertyView() -> (SourceRig_SinglePropertyView:Single Property View Object Reference)
     Get TargetRig_SinglePropertyView() -> (TargetRig_SinglePropertyView:Single Property View Object Reference)
    ]
    NewObject=Self-Reference()
   )
   > SetBrushColor(
    self=Get UnmappedSwatch_Border() -> (UnmappedSwatch_Border:Border Object Reference)
    InBrushColor=Get Unmapped Swatch Color() -> (Unmapped Swatch Color:Linear Color Structure)
   )
   > n13 = SetBrushColor(
    self=Get DuplicateSwatch_Border() -> (DuplicateSwatch_Border:Border Object Reference)
    InBrushColor=Get Duplicate Swatch Color() -> (Duplicate Swatch Color:Linear Color Structure)
   )
   Add Unique(
    execute=@n12.LoopBody
    TargetArray:Array of Animatable Creature Object References=Get AC files Array() -> (AC files Array:Array of Animatable Creature Object References)
    NewItem:"Animatable Creature Object Reference (by ref)"=@n12
   )
   "On Check State Changed (EnableMirroring_Checkbox)"!() -> (OutputDelegate, then, bIsChecked:bool) {is=title K2Node_ComponentBoundEvent}
  ]}
 Filter Rows and Update Statistics {nodes [
   Filter Rows and Update Statistics!() -> (then, Ignore None Rows:bool=~) {is=title K2Node_FunctionEntry}
   > n1 = Sequence!(execute) -> (then_0, then_1)
   > n2 = "For Each Loop (Map)"(MapPin:Map of Strings to EUW Mapping Row Object References=Get SourceControls()) -> (ValuePin:EUW Mapping Row Object Reference)
   > n3 = Set Current Mapping Row(Current Mapping Row:EUW Mapping Row Object Reference=@n2.ValuePin) -> (Output_Get:EUW Mapping Row Object Reference)
   > n4 = Is Duplicate Row!(execute, self, Mapping Row:EUW Mapping Row Object Reference=@n3) -> (then, ReturnValue:bool)
   > n5 = Is Unmapped Row!(
    execute
    self
    Mapping Row:EUW Mapping Row Object Reference=Get Current Mapping Row() -> (Current Mapping Row:EUW Mapping Row Object Reference)
   ) -> (then, ReturnValue:bool)
   > n6 = Branch(Condition=OR Boolean(A=@n5, B=@n4))
   > SetVisibility(self=Get Current Mapping Row(), InVisibility=Visible)
   n7 = Branch(
    execute=@n6.else
    Condition=IsChecked(self=Get OnlyUnresolved_Checkbox() -> (OnlyUnresolved_Checkbox:Editor Utility Check Box Object Reference))
   )
   n8 = Get Current Mapping Row()
   > SetVisibility(self=@n8, InVisibility=Collapsed)
   SetVisibility(execute=@n7.else, self=@n8, InVisibility=Visible)
   Update Statistics(execute=@n1.then_1)
   > Set Controls are not mapped(Controls are not mapped:bool=false) -> (Output_Get:bool)
   > n9 = "For Each Loop (Map)"(MapPin=Get SourceControls())
   > n10 = Set Current Mapping Row(Current Mapping Row=@n9.ValuePin)
   > Branch(Condition="Equal (Enum)"(
     A:ESlateVisibility Enum=Get Visibility!(self:Widget Object Reference=@n10) -> (Visibility:ESlateVisibility Enum=Visible)
     B:ESlateVisibility Enum=Visible
    ))
   > Set Controls are not mapped(Controls are not mapped=true)
   n11 = Branch(execute=@n9.CompletedPin, Condition=Get Controls are not mapped() -> (Controls are not mapped:bool))
   n12 = Get TextBox() -> (TextBox:Horizontal Box Object Reference)
   > Set Visibility(Visibility=Collapsed, self=@n12)
   n13 = Get EditorUtilityScrollBox_1078() -> (EditorUtilityScrollBox_1078:Editor Utility Scroll Box Object Reference)
   > Set Visibility(Visibility=Visible, self=@n13)
   Set Visibility(execute=@n11.else, Visibility=Visible, self=@n12)
   > Set Visibility(Visibility=Collapsed, self=@n13)
   n14 = GetDisplayName(Object=Get TargetRig())
   n15 = "To String (Integer)"(InInt=Length(TargetArray:Array of Strings=Get Sorted SourceControls() -> (Sorted SourceControls:Array of Strings)))
   > Set Text(
    Text="To Text (String)"(InString=Append!(A="All ", B=@n15, C:String=" controls have been mapped to ", D:String=@n14) -> (ReturnValue))
    self=Get Complete_Text() -> (Complete_Text:Text Object Reference)
   )
  ]}
 Is Duplicate Row {nodes [
   n1 = Is Duplicate Row!() -> (then, Mapping Row:EUW Mapping Row Object Reference) {is=title K2Node_FunctionEntry}
   n2 = Get RowBackground!(self:EUW Mapping Row Object Reference=@n1) -> (RowBackground:Border Object Reference)
   > Return Node!(
    execute
    ReturnValue:bool="Equal (LinearColor)"(
     A:Linear Color Structure=Get BrushColor!(self:Border Object Reference=@n2) -> (BrushColor:Linear Color Structure)
     B:Linear Color Structure=Get Duplicate Row Color() -> (Duplicate Row Color:Linear Color Structure)
    ) -> (ReturnValue:bool=~) {is=title K2Node_PromotableOperator}
   )
  ]}
 Is Unmapped Row {nodes [
   n1 = Is Unmapped Row!() -> (then, Mapping Row:EUW Mapping Row Object Reference) {is=title K2Node_FunctionEntry}
   n2 = Get TargetControlComboBox(self=@n1)
   > Return Node(ReturnValue="integer <= integer"(A=GetSelectedIndex!(self:"Searchable Combo Box (String) Object Reference"=@n2) -> (ReturnValue:int=0)))
  ]}
 IsInteractable {nodes [
   IsInteractable() {is=title K2Node_FunctionEntry}
   > Return Node(ReturnValue=false)
  ]}
 Mirror Suggested Controls {nodes [
   Mirror Suggested Controls!() -> (then, target Control Name:String) {is=title K2Node_FunctionEntry}
   > n1 = Sequence!(execute) -> (then_0, then_1, then_2:Exec, then_3:Exec)
   > n2 = Replace L or R with vise versa!(execute, self, "Control Name  to Change":String=Get target Control Name() -> (target Control Name:String)) -> (then, "Mirrored  Ctrl":String)
   > Set Mirrored Target Control(Mirrored Target Control:String=@n2) -> (Output_Get:String)
   > n3 = "For Each Loop (Map)"(MapPin=Get SourceControls())
   n4 = Get TargetControlComboBox(self=@n3.ValuePin)
   > Branch(Condition=Contains(
     SearchIn=GetSelectedOption!(self:"Searchable Combo Box (String) Object Reference"=@n4) -> (ReturnValue:String)
     Substring=Get target Control Name()
    ))
   > Set Temp Source Control Mapped(Temp Source Control Mapped:String=@n3) -> (Output_Get:String)
   n5 = Replace L or R with vise versa(execute=@n1.then_1, "Control Name  to Change"=Get Temp Source Control Mapped() -> (Temp Source Control Mapped:String))
   > Set Mirrored Source Control(Mirrored Source Control:String=@n5) -> (Output_Get:String)
   n6 = "For Each Loop (Set)"!(execute:Exec=@n1.then_2, SetPin:Set of Strings=Get TargetControls(), BreakPin:Exec) -> (then:Exec, ValuePin:String, CompletedPin:Exec) {is=title K2Node_SetForEach}
   > Branch(Condition="Equal Exactly (String)"(A=@n6, B=Get Mirrored Target Control() -> (Mirrored Target Control:String)))
   > n7 = "For Each Loop (Map)"(MapPin=Get SourceControls())
   > Branch(Condition="Equal Exactly (String)"(A=@n7, B=Get Mirrored Source Control() -> (Mirrored Source Control:String)))
   > n8 = Is Unmapped Row(Mapping Row=@n7.ValuePin) {is=title K2Node_CallFunction}
   > Branch(Condition=@n8)
   > SetSelectedOption(self=Get TargetControlComboBox(self=@n7.ValuePin), Option=Get Mirrored Target Control())
   Return Node!(execute=@n1.then_3)
  ]}
 Parse Into An Array {nodes [
   n1 = Parse Into An Array!() -> (then, input name to check:String, Substring:String) {is=title K2Node_FunctionEntry}
   n2 = Contains(SearchIn=@n1, Substring=@n1.Substring)
   > n3 = Branch(Condition=@n2)
   > Return Node!(execute, Output Parsed Array:Array of Strings=ParseIntoArray(SourceString=@n1, Delimiter=@n1.Substring), ReturnValue:bool=@n2)
   Return Node(execute=@n3.else, ReturnValue=@n2)
  ]}
 Populate Source Ctrls and Alt Row Colors {nodes [
   Populate Source Ctrls and Alt Row Colors() {is=title K2Node_FunctionEntry}
   > n1 = Sequence!(execute) -> (then_0, then_1, then_2)
   > Clear(TargetArray=Get Sorted SourceControls()) {is=title K2Node_CallArrayFunction}
   > Clear(TargetMap:Map of Strings to EUW Mapping Row Object References=Get SourceControls()) {is=title K2Node_CallFunction}
   > Clear(TargetArray=Get SortedChannels() -> (SortedChannels:Array of Strings)) {is=title K2Node_CallArrayFunction}
   n2 = Execute Python Script!(
    execute=@n1.then_1
    PythonScript="import unreal\r\n\r\nbp = unreal.load_object(name=in_path, outer=None)\r\nrig = bp.create_control_rig()\r\nrig.execute(\"Construction\")\r\nout_rig = rig\r\nout_hierarchy = rig.get_hierarchy()\r\nout_controls = out_hierarchy.get_controls(traverse=False)\r\n\r\n##First pass: collect AnimationChannel children per parent\r\nanim_channel_children = {}\r\nfor key in out_controls:\r\n  settings = out_hierarchy.get_control_settings(key)\r\n  if settings.animation_type == unreal.RigControlAnimationType.ANIMATION_CHANNEL:\r\n      parents = out_hierarchy.get_parents(key)\r\n      if parents:\r\n          parent_name = str(parents[0].name)\r\n          child_name = str(key.name).split(\"/\")[-1]\r\n          anim_channel_children.setdefault(parent_name, []).append(child_name)\r\n\r\n##Second pass: build channel strings\r\nchannel_strings = []\r\nfor key in out_controls:\r\n  settings = out_hierarchy.get_control_settings(key)\r\n  if settings.animation_type == unreal.RigControlAnimationType.ANIMATION_CHANNEL:\r\n      channel_strings.append(\"\")\r\n      continue\r\n\r\n  ctrl_name = str(key.name)\r\n  channels = [ctrl_name] + anim_channel_children.get(ctrl_name, [])\r\n  channel_strings.append(\"|\".join(channels))\r\n\r\nout_channel_strings = channel_strings"
    in_path:String=Get Object Path String(Object=Get SourceRig())
   ) -> (
    then
    ReturnValue
    out_controls:Array of Rig Element Key Structures
    out_hierarchy:Rig Hierarchy Object Reference
    out_channel_strings:Array of Strings
   ) {inputs [in_path], outputs [out_controls, out_hierarchy, out_channel_strings]}
   > n3 = For Each Loop(Array:Array of Rig Element Key Structures=@n2.out_controls) -> (Array Element:Rig Element Key Structure)
   > n4 = GetController!(execute, self:Rig Hierarchy Object Reference=@n2.out_hierarchy, bCreateIfNeeded:bool=false) -> (then, ReturnValue:Rig Hierarchy Controller Object Reference)
   n5 = Break Rig Control Settings!(RigControlSettings:"Rig Control Settings Structure (by ref)"=GetControlSettings!(self:Rig Hierarchy Controller Object Reference=@n4, InKey:Rig Element Key Structure=@n3) -> (ReturnValue:Rig Control Settings Structure)) -> (
    AnimationType:ERigControlAnimationType Enum
    ControlType:ERigControlType Enum
    DisplayName:Name
    PrimaryAxis:ERigControlAxis Enum
    LimitEnabled:Array of Rig Control Limit Enabled Structures
    bDrawLimits:bool=~
    MinimumValue:Rig Control Value Structure
    MaximumValue:Rig Control Value Structure
    bShapeVisible:bool=~
    ShapeVisibility:ERigControlVisibility Enum
    ShapeName:Name
    ShapeColor:Linear Color Structure
    bIsTransientControl:bool=~
    ControlEnum:Enum Object Reference
    Customization:Rig Control Element Customization Structure
    DrivenControls:Array of Rig Element Key Structures
    bGroupWithParentControl:bool=~
    bRestrictSpaceSwitching:bool=~
    FilteredChannels:Array of ERigControlTransformChannel Enums
    PreferredRotationOrder:EEulerRotationOrder Enum
    bUsePreferredRotationOrder:bool=~
    ShapeTransform:Transform
   ) {is=title K2Node_BreakStruct}
   > Branch(Condition=OR Boolean(
     A="Equal (Enum)"(A:ERigControlAnimationType Enum=@n5, B:ERigControlAnimationType Enum=AnimationControl)
     B="Equal (Enum)"(A=@n5, B=ProxyControl)
    ))
   > Add(
    TargetArray=Get Sorted SourceControls()
    NewItem="To String (Name)"(InName=Break Rig Element Key!(RigElementKey:"Rig Element Key Structure (by ref)"=@n3) -> (Type:ERigElementType Enum, Name:Name) {is=title K2Node_BreakStruct}.Name)
   ) {is=title K2Node_CallArrayFunction}
   > Add(
    TargetArray=Get SortedChannels()
    NewItem="Get (a ref)"(Array:Array of Strings=@n2.out_channel_strings, Dimension 1=@n3.Array Index) -> (Output:"String (by ref)")
   )
   n6 = For Each Loop(Exec=@n1.then_2, Array:Array of Strings=Get Sorted SourceControls()) -> (Array Element:String)
   Branch(execute=@n6.Completed, Condition=Get Is Pre Construct Event())
   > Set Row Colors(Ignore None Rows=true)
   > Filter Rows and Update Statistics(Ignore None Rows=false) {is=title K2Node_CallFunction}
   n7 = Construct Object from Class(execute=@n6.LoopBody, Class=&/EUW_MappingRow.EUW_MappingRow_C) -> (ReturnValue:EUW Mapping Row Object Reference)
   > n8 = Set Mapping Row(Mapping Row:EUW Mapping Row Object Reference=@n7) -> (Output_Get:EUW Mapping Row Object Reference)
   > Set SourceControlName!(execute, SourceControlName:String=@n6, self:EUW Mapping Row Object Reference=@n8) -> (then, Output_Get:String)
   > Set ParentMappingsWidget!(execute, ParentMappingsWidget:EUW Mappings Widget Object Reference=Self-Reference(), self:EUW Mapping Row Object Reference=@n8) -> (then, Output_Get:EUW Mappings Widget Object Reference)
   > Set Sorted Row Index!(
    execute
    Sorted Row Index:int=@n6.Array Index
    self:EUW Mapping Row Object Reference=Get Mapping Row() -> (Mapping Row:EUW Mapping Row Object Reference)
   ) -> (then, Output_Get:int=0)
   > Bind Event to On Selection Changed Dispatcher(self:EUW Mapping Row Object Reference=@n7, Delegate=Create Event()) {is=title K2Node_AddDelegate}
   > AddChild(self=Get Source_VerticalBox() -> (Source_VerticalBox:Vertical Box Object Reference), Content=Get Mapping Row())
   n9 = Get SortedChannels()
   > Set SourceChannels!(
    execute
    SourceChannels:Array of Strings=ParseIntoArray(SourceString="Get (a ref)"(Array=@n9, Dimension 1=@n6.Array Index), Delimiter=|)
    self:EUW Mapping Row Object Reference=Get Mapping Row()
   ) -> (then, Output_Get:Array of Strings)
   > Add(
    TargetMap:Map of Strings to EUW Mapping Row Object References=Get SourceControls()
    Key=@n6
    Value:"EUW Mapping Row Object Reference (by ref)"=Get Mapping Row()
   ) {is=title K2Node_CallFunction}
  ]}
 Populate Target Rig when AC is changed {nodes [
   Populate Target Rig when AC is changed() {is=title K2Node_FunctionEntry}
   > n1 = Sequence!(execute) -> (then_0, then_1, then_2, then_3:Exec)
   > n2 = LoadAsset_Blocking(Asset=Get Target_AC())
   > n3 = For Each Loop(Array:Array of Animatable Creature Object References=Get AC files Array()) -> (Array Element:Animatable Creature Object Reference)
   > Branch(Condition="Equal Exactly (String)"(A=GetDisplayName(Object=@n3), B=GetDisplayName(Object=@n2)))
   n4 = Get Control Rig(self=@n3)
   > n5 = LoadAsset(AssetPath=Split(
     SourceString="To String (Control Rig Soft Reference)"!(InRef:"Control Rig Asset Soft Reference Structure (by ref)"=@n4) -> (ReturnValue:String)
     InStr=.
     SearchDir=FromEnd
    ))
   > n6 = CastToControlRigBlueprint!(execute, Object:Object Reference=@n5) -> (CastSucceeded:Exec, CastFailed:Exec, AsControlRigBlueprint:Control Rig Blueprint Object Reference)
   > Set TargetRig(TargetRig:Control Rig Blueprint Object Reference=@n6) -> (Output_Get:Control Rig Blueprint Object Reference)
   Set bIsEnabled(execute=@n1.then_1, bIsEnabled=true, self=Get TargetRig_SinglePropertyView())
   > Call ED Target Rig Value Changed!(execute:Exec, self:EUW Mappings Widget Object Reference, Asset not none:bool) -> (then:Exec) {is=title K2Node_CallDelegate}
   n7 = Is Valid(exec=@n1.then_2, InputObject=Get TargetRig())
   Set bIsEnabled(execute=@n7.Is Not Valid, bIsEnabled=false, self=Get ImportDataTableFile())
   Set bIsEnabled(execute=@n7.Is Valid, bIsEnabled=true, self=Get ImportDataTableFile())
   Return Node!(execute=@n1.then_3)
  ]}
 Replace L or R with vise versa {nodes [
   Replace L or R with vise versa!() -> (then, "Control Name  to Change":String) {is=title K2Node_FunctionEntry}
   > n1 = Sequence()
   > n2 = Execute Python Script!(
    execute
    PythonScript="import unreal \r\n\r\nrenamed_ctrl = ''\r\nif old in ctrl_to_rename:\r\n    renamed_ctrl = ctrl_to_rename.replace(old, new)\r\n\r\nelif new in ctrl_to_rename:\r\n    renamed_ctrl = ctrl_to_rename.replace(new, old)\r\nelse:\r\n    renamed_ctrl = ctrl_to_rename"
    ctrl_to_rename:String="Get Control Name  to Change"() -> ("Control Name  to Change":String)
    old:String=Get underscore l() -> (underscore l:String)
    new:String=Get underscore r() -> (underscore r:String)
   ) -> (then, ReturnValue, renamed_ctrl:String) {inputs [ctrl_to_rename, old, new], outputs [renamed_ctrl]}
   > n3 = Set Mirrored Source Control(Mirrored Source Control=@n2.renamed_ctrl)
   > n4 = Execute Python Script(
    PythonScript="import unreal \r\n\r\nrenamed_ctrl = ''\r\nif old in ctrl_to_rename:\r\n    renamed_ctrl = ctrl_to_rename.replace(old, new)\r\n\r\nelif new in ctrl_to_rename:\r\n    renamed_ctrl = ctrl_to_rename.replace(new, old)\r\nelse:\r\n    renamed_ctrl = ctrl_to_rename"
    ctrl_to_rename=@n3
    old=Get right() -> (right:String)
    new=Get left() -> (left:String)
   ) {inputs [ctrl_to_rename, old, new], outputs [renamed_ctrl]}
   > n5 = Set Mirrored Source Control(Mirrored Source Control=@n4.renamed_ctrl)
   > n6 = Execute Python Script(
    PythonScript="import unreal \r\n\r\nrenamed_ctrl = ''\r\nif old in ctrl_to_rename:\r\n    renamed_ctrl = ctrl_to_rename.replace(old, new)\r\n\r\nelif new in ctrl_to_rename:\r\n    renamed_ctrl = ctrl_to_rename.replace(new, old)\r\nelse:\r\n    renamed_ctrl = ctrl_to_rename"
    ctrl_to_rename=@n5
    old=Get Backslash L() -> (Backslash L:String)
    new=Get Backslash R() -> (Backslash R:String)
   ) {inputs [ctrl_to_rename, old, new], outputs [renamed_ctrl]}
   > Set Mirrored Source Control(Mirrored Source Control=@n6.renamed_ctrl)
   n7 = Execute Python Script(
    execute=@n1.then_1
    PythonScript="import unreal \r\n\r\nrenamed_ctrl = ''\r\nif old in ctrl_to_rename:\r\n    renamed_ctrl = ctrl_to_rename.replace(old, new)\r\n\r\nelif new in ctrl_to_rename:\r\n    renamed_ctrl = ctrl_to_rename.replace(new, old)\r\nelse:\r\n    renamed_ctrl = ctrl_to_rename"
    ctrl_to_rename=Get Mirrored Source Control()
    old=Get L Backslash() -> (L Backslash:String)
    new=Get R backslash() -> (R backslash:String)
   ) {inputs [ctrl_to_rename, old, new], outputs [renamed_ctrl]}
   > n8 = Set Mirrored Source Control(Mirrored Source Control=@n7.renamed_ctrl)
   > n9 = Execute Python Script(
    PythonScript="import unreal \r\n\r\nrenamed_ctrl = ''\r\nif old in ctrl_to_rename:\r\n    renamed_ctrl = ctrl_to_rename.replace(old, new)\r\n\r\nelif new in ctrl_to_rename:\r\n    renamed_ctrl = ctrl_to_rename.replace(new, old)\r\nelse:\r\n    renamed_ctrl = ctrl_to_rename"
    ctrl_to_rename=@n8
    old=Get R with 2 spaces() -> (R with 2 spaces:String)
    new=Get L with 2 spaces() -> (L with 2 spaces:String)
   ) {inputs [ctrl_to_rename, old, new], outputs [renamed_ctrl]}
   > Set Mirrored Source Control(Mirrored Source Control=@n9.renamed_ctrl)
   n10 = Branch(execute=@n1.then_2, Condition=EndsWith(SourceString=Get Mirrored Source Control(), InSuffix=" L", SearchCase=CaseSensitive))
   > n11 = Replace R or L!(
    execute
    self
    Source Ctrl:String=Get Mirrored Source Control()
    substring to look for:String=" L"
    what to replace to:String=" R"
    search from End:bool=true
   ) -> (then, Mirrored Source Ctrl:String)
   > Set Mirrored Source Control(Mirrored Source Control=@n11)
   Branch(execute=@n10.else, Condition=EndsWith(SourceString=Get Mirrored Source Control(), InSuffix=" R", SearchCase=CaseSensitive))
   > n12 = Replace R or L(Source Ctrl=Get Mirrored Source Control(), substring to look for=" R", what to replace to=" L", search from End=true)
   > Set Mirrored Source Control(Mirrored Source Control=@n12)
   Return Node!(execute=@n1.then_3, "Mirrored  Ctrl":String=Get Mirrored Source Control())
  ]}
 Replace R or L {nodes [
   n1 = Replace R or L!() -> (then, Source Ctrl:String, substring to look for:String, what to replace to:String, search from End:bool=~) {is=title K2Node_FunctionEntry}
   > n2 = Sequence!(execute) -> (then_0, then_1, then_2)
   > Branch(Condition=Contains(
     SearchIn=Get Source Ctrl() -> (Source Ctrl:String)
     Substring=Get substring to look for() -> (substring to look for:String)
     bUseCase=true
     bSearchFromEnd=@n1.search from End
    ))
   > n3 = ReplaceInline!(
    execute
    SourceString:"String (by ref)"=Get Source Ctrl()
    SearchText:String=Get substring to look for()
    ReplacementText:String=Get what to replace to() -> (what to replace to:String)
    SearchCase:ESearchCase Enum=CaseSensitive
   ) -> (then, ReturnValue:int=0)
   > Set NewLocalVar_0(NewLocalVar_0:int=@n3) -> (Output_Get:int=0)
   n4 = Branch(execute=[@n2.then_2, @n5.then], Condition="integer > integer"(A=Get NewLocalVar_0() -> (NewLocalVar_0:int=0)))
   > Set NewLocalVar(NewLocalVar:String=Replace(SourceString=Get Source Ctrl(), From=Get substring to look for(), To=Get what to replace to(), SearchCase=CaseSensitive)) -> (Output_Get:String)
   > n5 = Set NewLocalVar_0(NewLocalVar_0=int - int(A=Get NewLocalVar_0() -> (NewLocalVar_0=0))) -> (Output_Get=0)
   Return Node!(execute=@n4.else, Mirrored Source Ctrl:String=Get NewLocalVar() -> (NewLocalVar:String))
  ]}
 "Reset Target Ctrls Mapping, Populate Target Ctrl Set" {nodes [
   "Reset Target Ctrls Mapping, Populate Target Ctrl Set"() {is=title K2Node_FunctionEntry}
   > n1 = Sequence()
   n2 = Execute Python Script!(
    execute=@n1.then_1
    PythonScript="import unreal\r\n\r\nbp = unreal.load_object(name=in_path, outer=None)\r\nout_controls = []\r\nout_channel_strings = []\r\nif bp:\r\n  rig = bp.create_control_rig()\r\n  rig.execute(\"Construction\")\r\n  out_hierarchy = rig.get_hierarchy()\r\n  out_controls = out_hierarchy.get_controls(traverse=False)\r\n\r\n  ## First pass: collect AnimationChannel children per parent\r\n  anim_channel_children = {}\r\n  for key in out_controls:\r\n      settings = out_hierarchy.get_control_settings(key)\r\n      if settings.animation_type == unreal.RigControlAnimationType.ANIMATION_CHANNEL:\r\n          parents = out_hierarchy.get_parents(key)\r\n          if parents:\r\n              parent_name = str(parents[0].name)\r\n              child_name = str(key.name).split(\"/\")[-1]\r\n              anim_channel_children.setdefault(parent_name, []).append(child_name)\r\n\r\n  ## Second pass: build channel strings\r\n  for key in out_controls:\r\n      settings = out_hierarchy.get_control_settings(key)\r\n      if settings.animation_type == unreal.RigControlAnimationType.ANIMATION_CHANNEL:\r\n          out_channel_strings.append(\"\")\r\n          continue\r\n\r\n      ctrl_name = str(key.name)\r\n      channels = [ctrl_name] + anim_channel_children.get(ctrl_name, [])\r\n      out_channel_strings.append(\"|\".join(channels))"
    in_path:String=Get Object Path String(Object=Get TargetRig())
   ) -> (then, ReturnValue, out_controls:Array of Rig Element Key Structures, out_channel_strings:Array of Strings) {inputs [in_path], outputs [out_controls, out_channel_strings]}
   > n3 = For Each Loop!(Exec, Array:Array of Rig Element Key Structures=@n2.out_controls) -> (LoopBody, Array Element_Type:ERigElementType Enum, Array Element_Name:Name, Array Index, Completed)
   n4 = "To String (Name)"(InName=@n3.Array Element_Name)
   > Add!(execute, TargetSet:Set of Strings=Get TargetControls(), NewItem:"String (by ref)"=@n4) -> (then)
   n5 = "Get (a ref)"(Array=@n2.out_channel_strings, Dimension 1=@n3.Array Index)
   > Branch(Condition="Not Equal Exactly (String)"(A=@n5))
   > Add!(
    execute
    TargetMap:Map of Strings to Strings=Get TargetChannelMap() -> (TargetChannelMap:Map of Strings to Strings)
    Key:"String (by ref)"=@n4
    Value:"String (by ref)"=@n5
   ) -> (then)
   n6 = "For Each Loop (Map)"(execute=@n1.then_2, MapPin=Get SourceControls())
   n7 = Get TargetControlComboBox(self=@n6.ValuePin)
   > ClearOptions(self:"Searchable Combo Box (String) Object Reference"=@n7)
   > n8 = "For Each Loop (Set)"(SetPin=Get TargetControls())
   > AddOption!(execute, self:"Searchable Combo Box (String) Object Reference"=@n7, Option:String=@n8) -> (then)
   n9 = Get TargetControls()
   Clear!(execute=@n1.then_0, TargetSet:Set of Strings=@n9) -> (then) {is=title K2Node_CallFunction}
   > Add!(execute, TargetSet:Set of Strings=@n9, NewItem:"String (by ref)"="<None>") -> (then)
   > Clear!(execute, TargetMap:Map of Strings to Strings=Get TargetChannelMap()) -> (then)
  ]}
 Return String Mapping {nodes [
   Return String Mapping() {is=title K2Node_FunctionEntry}
   > Clear(TargetMap=Get Created mapping() -> (Created mapping:Map of Strings to Strings))
   > n1 = "For Each Loop (Map)"(MapPin=Get SourceControls())
   > Add!(
    execute
    TargetMap:Map of Strings to Strings=Get Created mapping()
    Key:"String (by ref)"=@n1
    Value:"String (by ref)"=GetSelectedOption(self=Get TargetControlComboBox(self=@n1.ValuePin))
   ) -> (then)
   n2 = "For Each Loop (Map)"(execute=@n1.CompletedPin, MapPin:Map of Strings to Strings=Get extra_channels()) -> (ValuePin:String)
   > Add(TargetMap=Get Created mapping(), Key=@n2, Value=@n2.ValuePin)
   Return Node!(execute=@n2.CompletedPin, Final Mapping:Map of Strings to Strings=Get Created mapping())
  ]}
 Set Row Colors {nodes [
   n1 = Set Row Colors!() -> (then, Ignore None Rows:bool=~) {is=title K2Node_FunctionEntry}
   > Clear(TargetMap:Map of Strings to Mapping Row Array Struct Structures=Get MappedControls() -> (MappedControls:Map of Strings to Mapping Row Array Struct Structures))
   > n2 = Sequence()
   Return Node!(execute=@n2.then_2)
   n3 = "For Each Loop (Map)"(execute=@n2.then_1, MapPin:Map of Strings to Mapping Row Array Struct Structures=Get MappedControls()) -> (ValuePin:Mapping Row Array Struct Structure)
   n4 = Break Mapping Row Array Struct!(MappingRowArray_Struct:"Mapping Row Array Struct Structure (by ref)"=@n3.ValuePin) -> (ArrayofMappingRows_13_#18:Array of EUW Mapping Row Object References) {is=title K2Node_BreakStruct}
   > Branch(Condition="integer > integer"(A=Length(TargetArray:Array of EUW Mapping Row Object References=@n4), B=1))
   > n5 = For Each Loop!(Exec, Array:Array of EUW Mapping Row Object References=@n4) -> (LoopBody, Array Element:EUW Mapping Row Object Reference, Array Index, Completed)
   > Set BrushColor!(execute, BrushColor:Linear Color Structure=Get Duplicate Row Color(), self:Border Object Reference=Get RowBackground(self=@n5)) -> (then, Output_Get:Linear Color Structure)
   n6 = "For Each Loop (Map)"(execute=@n2.then_0, MapPin:Map of Strings to EUW Mapping Row Object References=Get SourceControls()) -> (ValuePin:EUW Mapping Row Object Reference)
   > n7 = Set Current Mapping Row(Current Mapping Row=@n6.ValuePin)
   > n8 = Is Unmapped Row(Mapping Row=@n7)
   > n9 = Branch(Condition=@n8)
   > n10 = Branch(Condition=@n1)
   n11 = Get Sorted Row Index!(self:EUW Mapping Row Object Reference=Get Current Mapping Row()) -> (Sorted Row Index:int=0)
   n12 = Branch(execute=[@n22.then, @n10.then], Condition="Equal (Integer)"(A="% (Integer)"(A=@n11, B=2) -> (ReturnValue=0)))
   > n13 = Set Current Row Color(Current Row Color:Linear Color Structure=Get Assigned Row Color 0() -> (Assigned Row Color 0:Linear Color Structure)) -> (Output_Get:Linear Color Structure)
   Set BrushColor(
    execute=[@n13.then, @n14.then]
    BrushColor=Get Current Row Color() -> (Current Row Color:Linear Color Structure)
    self=Get RowBackground(self=Get Current Mapping Row())
   )
   n14 = Set Current Row Color(execute=@n12.else, Current Row Color=Get Assigned Row Color 1() -> (Assigned Row Color 1:Linear Color Structure))
   Set BrushColor(
    execute=@n10.else
    BrushColor=Get Unmapped Row Color() -> (Unmapped Row Color:Linear Color Structure)
    self=Get RowBackground(self=Get Current Mapping Row())
   )
   n15 = Get Current Mapping Row()
   n16 = Set Current Selected Option(execute=@n9.else, Current Selected Option:String=GetSelectedOption(self=Get TargetControlComboBox(self=@n15))) -> (Output_Get:String)
   > n17 = Branch(Condition=Contains!(TargetMap:Map of Strings to Mapping Row Array Struct Structures=Get MappedControls(), Key:"String (by ref)"=@n16) -> (ReturnValue))
   n18 = Get Current Selected Option() -> (Current Selected Option:String)
   n19 = Get MappedControls()
   > n20 = Set Local Mapping Row Array Copy(Local Mapping Row Array Copy:Array of EUW Mapping Row Object References=Break Mapping Row Array Struct(MappingRowArray_Struct=Find(TargetMap:Map of Strings to Mapping Row Array Struct Structures=@n19, Key=@n18) -> (Value:Mapping Row Array Struct Structure))) -> (Output_Get:Array of EUW Mapping Row Object References)
   > Add(TargetArray:Array of EUW Mapping Row Object References=@n20, NewItem:"EUW Mapping Row Object Reference (by ref)"=Get Current Mapping Row()) {is=title K2Node_CallArrayFunction}
   > Add(
    TargetMap:Map of Strings to Mapping Row Array Struct Structures=Get MappedControls()
    Key=Get Current Selected Option()
    Value:"Mapping Row Array Struct Structure (by ref)"=Make Mapping Row Array Struct!(ArrayofMappingRows_13_#18:Array of EUW Mapping Row Object References=@n20) -> (MappingRowArray_Struct:Mapping Row Array Struct Structure) {is=title K2Node_MakeStruct}
   ) {is=title K2Node_CallFunction}
   n21 = Get Current Mapping Row()
   n22 = Add(
    execute=@n17.else
    TargetMap=Get MappedControls()
    Key=Get Current Selected Option()
    Value=Make Mapping Row Array Struct(ArrayofMappingRows_13_#18=Make Array("[0]":EUW Mapping Row Object Reference=@n21) -> (Array:Array of EUW Mapping Row Object References))
   )
  ]}
 Target or Source Rig Updated {nodes [
   Target or Source Rig Updated() {is=title K2Node_FunctionEntry}
   > n1 = Sequence!(execute) -> (then_0, then_1, then_2, then_3:Exec, then_4:Exec, then_5:Exec)
   > Clear Source Ctrls()
   Populate Source Ctrls and Alt Row Colors(execute=@n1.then_1) {is=title K2Node_CallFunction}
   "Reset Target Ctrls Mapping, Populate Target Ctrl Set"(execute=@n1.then_2) {is=title K2Node_CallFunction}
   Auto Suggest Mapping of Controls(execute=@n1.then_3)
   Call ED Target Rig Value Changed(execute=@n1.then_5, Asset not none=IsValid(Object=Get TargetRig()))
   Set Row Colors(execute=@n1.then_4, Ignore None Rows=false) {is=title K2Node_CallFunction}
   > Filter Rows and Update Statistics(Ignore None Rows=false)
  ]}
 Update Mapping From Data Table {nodes [
   Update Mapping From Data Table() {is=title K2Node_FunctionEntry}
   > "Reset Target Ctrls Mapping, Populate Target Ctrl Set"()
   > Updating Mapping Rows with Data Table()
   > Set Row Colors(Ignore None Rows=false)
   > Filter Rows and Update Statistics(Ignore None Rows=false)
  ]}
 "Update Row Colors, Statistics, and Mirror options" {nodes [
   "Update Row Colors, Statistics, and Mirror options"!() -> (then, MappingRow:EUW Mapping Row Object Reference, Selected Control:String) {is=title K2Node_FunctionEntry}
   > n1 = Sequence!(execute) -> (then_0, then_1, then_2)
   > Set Row Colors(Ignore None Rows=false)
   Branch(
    execute=@n1.then_1
    Condition=IsChecked(self=Get EnableMirroring_Checkbox() -> (EnableMirroring_Checkbox:Editor Utility Check Box Object Reference))
   )
   > Mirror Suggested Controls!(execute, self, target Control Name:String=Get Selected Control() -> (Selected Control:String)) -> (then) {is=title K2Node_CallFunction}
   Set Row Colors(execute=@n1.then_2, Ignore None Rows=false)
   > Update Statistics()
  ]}
 Update Statistics {nodes [
   Update Statistics() {is=title K2Node_FunctionEntry}
   > Set Unmapped Count(Unmapped Count:int=0) -> (Output_Get:int=0)
   > Set Duplicates Count(Duplicates Count:int=0) -> (Output_Get:int=0)
   > n1 = Sequence!(execute) -> (then_0, then_1)
   > n2 = "For Each Loop (Map)"(MapPin=Get SourceControls())
   > n3 = Is Unmapped Row(Mapping Row=@n2.ValuePin)
   > n4 = Branch(Condition=@n3)
   > Set Unmapped Count(Unmapped Count=int + int(A=Get Unmapped Count() -> (Unmapped Count:int=0))) -> (Output_Get=0)
   n5 = Is Duplicate Row(execute=@n4.else, Mapping Row=@n2.ValuePin) {is=title K2Node_CallFunction}
   > Branch(Condition=@n5)
   > Set Duplicates Count(Duplicates Count=int + int(A=Get Duplicates Count() -> (Duplicates Count:int=0))) -> (Output_Get=0)
   n6 = Get Unmapped Count() -> (Unmapped Count=0)
   Set Text(execute=@n1.then_1, Text="To Text (Integer)"(Value=@n6), self=Get UnmappedNumber_Text() -> (UnmappedNumber_Text:Text Object Reference))
   n7 = Get Duplicates Count() -> (Duplicates Count=0)
   > Set Text(Text="To Text (Integer)"(Value=@n7), self=Get DuplicateNumber_Text() -> (DuplicateNumber_Text:Text Object Reference))
   n8 = int - int(A=Length(TargetArray:Array of Strings=Get Sorted SourceControls()), B=@n6)
   > Set Text(Text="To Text (Integer)"(Value=int - int(A=@n8, B=@n7)), self=Get MappedNumber_Text() -> (MappedNumber_Text:Text Object Reference))
  ]}
 Updating Mapping Rows with Data Table {nodes [
   Updating Mapping Rows with Data Table() {is=title K2Node_FunctionEntry}
   > n1 = Sequence!(execute) -> (then_0, then_1, then_2:Exec)
   > Clear(TargetMap:Map of Strings to Strings=Get extra_channels())
   > n2 = GetDataTableRowNames(Table=Get Data_Table())
   > n3 = For Each Loop(Array:Array of Names=@n2) -> (Array Element:Name)
   > n4 = Get Data Table Row(DataTable=Get Data_Table(), RowName=@n3) -> (ReturnValue:Str to Str Mapping Struct Structure)
   n5 = Break Str To Str Mapping Struct!(Str_to_Str_Mapping_Struct:"Str to Str Mapping Struct Structure (by ref)"=@n4) -> (
    mapattributes_23_#11:bool=~
    SourceCtrl_6_#12:String
    TargetCtrl_8_#13:String
    SourceRig_15_#14:String
    TargetRig_16_#15:String
    SourceChannel_19_#16:String
    TargetChannel_25_#17:String
   ) {is=title K2Node_BreakStruct}
   > Branch(Condition="Not Equal Exactly (String)"(A=@n5.SourceChannel_19_#16))
   > Add(
    TargetMap:Map of Strings to Strings=Get extra_channels()
    Key=@n5.SourceChannel_19_#16
    Value:"String (by ref)"=@n5.TargetChannel_25_#17
   )
   n6 = GetDataTableRowNames(execute=@n1.then_1, Table=Get Data_Table())
   > n7 = For Each Loop(Array=@n6)
   > n8 = Get Data Table Row(DataTable=Get Data_Table(), RowName=@n7)
   n9 = Break Str To Str Mapping Struct(Str_to_Str_Mapping_Struct=@n8) -> (mapattributes_23_#11=~)
   > Branch(Condition="Equal, Case Insensitive (String)"(A=@n9.SourceChannel_19_#16))
   n10 = Find(TargetMap:Map of Strings to EUW Mapping Row Object References=Get SourceControls(), Key=@n9.SourceCtrl_6_#12) -> (Value:EUW Mapping Row Object Reference)
   > Branch(Condition=@n10.ReturnValue)
   > SetSelectedOption(self=Get TargetControlComboBox(self=@n10), Option=@n9.TargetCtrl_8_#13)
   > RefreshChannelRows(self=@n10, TargetCtrl=@n9.TargetCtrl_8_#13, bExpanded=@n9)
   Return Node(execute=@n1.then_2)
  ]}
]
variables [
 AC files Array {type="array<AnimatableCreature_C>", default []}
 animatable creatures {type="array<AnimatableCreature_C>", default []}
 Assigned Row Color 0 {type=LinearColor, default=<rgba 0.035 0.035 0.035 1>}
 Assigned Row Color 1 {type=LinearColor, default=<rgba 0.0226 0.0226 0.0226 1>}
 copy anim euw {type=EUW_RigReplacementTool_C, default=null}
 Current Mapping Row {type=EUW_MappingRow_C, default=null}
 Current Row Color {type=LinearColor, default=<rgba 0.0495 0.0564 0.151 1>}
 Data Table Location {type=string, default=""}
 Data Table Name {type=string, default=""}
 Data Table updated {default=()}
 Data_Table {type=DataTable, default=null}
 Duplicate Row Color {type=LinearColor, default=<rgba 0.1059 0.0864 0.0355 1>}
 Duplicate Swatch Color {type=LinearColor, default=<rgba 0.2309 0.1884 0.0774 1>}
 Duplicates Count {type=integer}
 ED Target Rig Value Changed {default=()}
 extra_channels {type="map<string>", default {}}
 Is Pre Construct Event {type=boolean}
 LevelSequenceBrowser {type=EUW_LevelSequenceBrowser_C, default=null}
 MappedControls {type="map<MappingRowArray_Struct>", default {}}
 Mapping Row {type=EUW_MappingRow_C, default=null}
 Mapping_Struct_Path {type=string, default="'/AnimInEdTools/CopyAnimBatch/Str_to_Str_Mapping_Struct' "}
 MappingLibrary {type=DataTable, default=/AnimInEdTools/RigReplacement/DataTables/DT_MappingLibrary}
 New Created Data Table Asset {type=DataTable, default=null}
 Sorted SourceControls {type="array<string>", default []}
 SourceControls {type="map<EUW_MappingRow_C>", default {}}
 SourceRig {type=ControlRigBlueprint, default=/EpicControlRig/Templates/MR_Biped_v01, class=~}
 Target_AC {type=AnimatableCreature_C, default=null}
 TargetChannelMap {type="map<string>", default {}}
 TargetControls {type="array<string>", default []}
 TargetRig {type=ControlRigBlueprint, default=null, class=~}
 Unmapped Count {type=integer}
 Unmapped Row Color {type=LinearColor, default=<rgba 0.1059 0.0258 0.0241 1>}
 Unmapped Swatch Color {type=LinearColor, default=<rgba 0.2726 0.0664 0.062 1>}
]
widgets {
 info {namedSlotCount=6, rootWidgetClass=&VerticalBox, widgetCount=50}
 widgets [
  Main_VerticalBox {widgetClassPath=&VerticalBox, widget=Main_VerticalBox, properties {bOverride_Cursor=true}}
  RigSetupExpandable {
   widgetClassPath=/Script/Blutility.EditorUtilityExpandableArea
   bIsVariable=true
   parent=Main_VerticalBox
   widget=RigSetupExpandable
   properties {bIsExpanded=true}
   slot {class=VerticalBoxSlot}
  }
  HorizontalBox_71 {widgetClassPath=&HorizontalBox, namedSlotHost=RigSetupExpandable, widget=HorizontalBox_71}
  TextBlock_87 {
   widgetClassPath=&TextBlock
   parent=HorizontalBox_71
   widget=TextBlock_87
   properties {
    text=Rig Setup
    font {
     bForceMonospaced=false
     bMaterialIsStencil=false
     fontMaterial=null
     fontObject=&Roboto
     letterSpacing=0
     monospacedWidth=1
     size=10.5
     skewAmount=0
     typefaceFontName=Regular
     outlineSettings {
      bApplyOutlineToDropShadows=false
      bMiteredCorners=false
      bSeparateFillAlpha=false
      outlineMaterial=null
      outlineSize=0
      outlineColor=<rgba 0 0 0 1>
     }
    }
   }
   slot {class=HorizontalBoxSlot, properties {horizontalAlignment=HAlign_Center, verticalAlignment=VAlign_Center}}
  }
  VerticalBox_68 {widgetClassPath=&VerticalBox, namedSlotHost=RigSetupExpandable, widget=VerticalBox_68}
  SourceRig_SinglePropertyView {
   widgetClassPath=&SinglePropertyView
   parent=VerticalBox_68
   widget=SourceRig_SinglePropertyView
   properties {propertyName=SourceRig, toolTipText=What rig do you want to retrieve animations FROM?}
   slot {class=VerticalBoxSlot, properties {padding=<ltrb 0 5 5 0>}}
  }
  TargetACFile {
   widgetClassPath=&SinglePropertyView
   parent=VerticalBox_68
   widget=TargetACFile
   properties {propertyName=Target_AC, toolTipText=What AC file you want to pull into the LevelSequence}
   slot {class=VerticalBoxSlot, properties {padding=<ltrb 4 5 0 10>}}
  }
  TargetRig_SinglePropertyView {
   widgetClassPath=&SinglePropertyView
   parent=VerticalBox_68
   widget=TargetRig_SinglePropertyView
   properties {bIsEnabled=false, bOverride_Cursor=true, propertyName=TargetRig, toolTipText=What Rig do you want the animations to transfer onto?}
   slot {class=VerticalBoxSlot, properties {padding=<ltrb 4 0 0 10>}}
  }
  Text_LoadingRig {
   widgetClassPath=&TextBlock
   bIsVariable=true
   parent=VerticalBox_68
   widget=Text_LoadingRig
   properties {text=Loading rig controls..., visibility=Collapsed}
   slot {class=VerticalBoxSlot, properties {horizontalAlignment=HAlign_Left}}
  }
  MappingDataExpandable {
   widgetClassPath=&EditorUtilityExpandableArea
   bIsVariable=true
   parent=Main_VerticalBox
   widget=MappingDataExpandable
   properties {bIsExpanded=true}
   slot {class=VerticalBoxSlot}
  }
  HorizontalBox_167 {widgetClassPath=&HorizontalBox, namedSlotHost=MappingDataExpandable, widget=HorizontalBox_167}
  TextBlock_164 {
   widgetClassPath=&TextBlock
   parent=HorizontalBox_167
   widget=TextBlock_164
   properties {
    text=Mapping Data
    font {
     bForceMonospaced=false
     bMaterialIsStencil=false
     fontMaterial=null
     fontObject=&Roboto
     letterSpacing=0
     monospacedWidth=1
     size=10.5
     skewAmount=0
     typefaceFontName=Regular
     outlineSettings {
      bApplyOutlineToDropShadows=false
      bMiteredCorners=false
      bSeparateFillAlpha=false
      outlineMaterial=null
      outlineSize=0
      outlineColor=<rgba 0 0 0 1>
     }
    }
   }
   slot {class=HorizontalBoxSlot, properties {verticalAlignment=VAlign_Center}}
  }
  HorizontalBox_168 {
   widgetClassPath=&HorizontalBox
   parent=HorizontalBox_167
   widget=HorizontalBox_168
   slot {class=HorizontalBoxSlot, properties {horizontalAlignment=HAlign_Right, padding=<ltrb 2 2 2 2>, size {sizeRule=Fill, value=1}}}
  }
  Create_Mapping_Btn {
   widgetClassPath=&EditorUtilityButton
   parent=HorizontalBox_168
   widget=Create_Mapping_Btn
   properties {toolTipText=This will save the Source and Target Control Mappings}
   slot {class=HorizontalBoxSlot}
  }
  RefreshContents_HorizontalBox_1 {
   widgetClassPath=&HorizontalBox
   parent=Create_Mapping_Btn
   widget=RefreshContents_HorizontalBox_1
   slot {class=ButtonSlot, properties {padding=<ltrb -6 2 -4 2>}}
  }
  SaveImage {
   widgetClassPath=&Image
   bIsVariable=true
   parent=RefreshContents_HorizontalBox_1
   widget=SaveImage
   properties {brush {
     bIsDynamicallyLoaded=false
     drawAs=Image
     imageType=FullColor
     mirroring=NoMirror
     resourceName=None
     resourceObject=&/T_SaveIcon
     tiling=NoTile
     imageSize=<20 20>
     margin=<ltrb 0 0 0 0>
     outlineSettings {
      bUseBrushTransparency=false
      roundingType=HalfHeightRadius
      width=0
      color {colorUseRule=UseColor_Specified, specifiedColor=<rgba 0 0 0 0>}
      cornerRadii=<0 0 0 0>
     }
     tintColor {colorUseRule=UseColor_Specified, specifiedColor=<rgba 1 1 1 1>}
     uVRegion {bIsValid=false, max=<0 0>, min=<0 0>}
    }}
   slot {class=HorizontalBoxSlot}
  }
  Refresh_1 {
   widgetClassPath=&TextBlock
   parent=RefreshContents_HorizontalBox_1
   widget=Refresh_1
   properties {
    text=Save Mappings
    font {
     bForceMonospaced=false
     bMaterialIsStencil=false
     fontMaterial=null
     fontObject=&Roboto
     letterSpacing=0
     monospacedWidth=1
     size=10.5
     skewAmount=0
     typefaceFontName=Regular
     outlineSettings {
      bApplyOutlineToDropShadows=false
      bMiteredCorners=false
      bSeparateFillAlpha=false
      outlineMaterial=null
      outlineSize=0
      outlineColor=<rgba 0 0 0 1>
     }
    }
   }
   slot {class=HorizontalBoxSlot, properties {verticalAlignment=VAlign_Center, padding=<ltrb 3 0 0 0>}}
  }
  VerticalBox_1 {widgetClassPath=&VerticalBox, namedSlotHost=MappingDataExpandable, widget=VerticalBox_1}
  ImportDataTableFile {
   widgetClassPath=&SinglePropertyView
   parent=VerticalBox_1
   widget=ImportDataTableFile
   properties {bIsEnabled=false, propertyName=Data_Table}
   slot {class=VerticalBoxSlot, properties {padding=<ltrb 0 5 0 5>}}
  }
  ControlMappingExpandable {
   widgetClassPath=&EditorUtilityExpandableArea
   bIsVariable=true
   parent=Main_VerticalBox
   widget=ControlMappingExpandable
   properties {bIsExpanded=true}
   slot {class=VerticalBoxSlot, properties {size {sizeRule=Fill, value=1}}}
  }
  WrapBox_0 {widgetClassPath=&WrapBox, namedSlotHost=ControlMappingExpandable, widget=WrapBox_0}
  TextBlock_242 {
   widgetClassPath=&TextBlock
   parent=WrapBox_0
   widget=TextBlock_242
   properties {
    text=Control Mapping
    font {
     bForceMonospaced=false
     bMaterialIsStencil=false
     fontMaterial=null
     fontObject=&Roboto
     letterSpacing=0
     monospacedWidth=1
     size=10.5
     skewAmount=0
     typefaceFontName=Regular
     outlineSettings {
      bApplyOutlineToDropShadows=false
      bMiteredCorners=false
      bSeparateFillAlpha=false
      outlineMaterial=null
      outlineSize=0
      outlineColor=<rgba 0 0 0 1>
     }
    }
   }
   slot {class=WrapBoxSlot, properties {verticalAlignment=VAlign_Center, padding=<ltrb 0 0 20 0>}}
  }
  HorizontalBox_3 {
   widgetClassPath=&HorizontalBox
   parent=WrapBox_0
   widget=HorizontalBox_3
   slot {class=WrapBoxSlot, properties {verticalAlignment=VAlign_Center, padding=<ltrb 10 5 10 5>}}
  }
  UnmappedSwatch_Border {
   widgetClassPath=&Border
   bIsVariable=true
   parent=HorizontalBox_3
   widget=UnmappedSwatch_Border
   properties {desiredSizeScale=<1.25 1>}
   slot {class=HorizontalBoxSlot, properties {padding=<ltrb 0 0 5 0>}}
  }
  UnmappedNumber_Text {
   widgetClassPath=&TextBlock
   bIsVariable=true
   parent=HorizontalBox_3
   widget=UnmappedNumber_Text
   properties {
    text=0
    font {
     bForceMonospaced=false
     bMaterialIsStencil=false
     fontMaterial=null
     fontObject=&Roboto
     letterSpacing=0
     monospacedWidth=1
     size=10.5
     skewAmount=0
     typefaceFontName=Regular
     outlineSettings {
      bApplyOutlineToDropShadows=false
      bMiteredCorners=false
      bSeparateFillAlpha=false
      outlineMaterial=null
      outlineSize=0
      outlineColor=<rgba 0 0 0 1>
     }
    }
   }
   slot {class=HorizontalBoxSlot}
  }
  Unmapped_Text {
   widgetClassPath=&TextBlock
   parent=HorizontalBox_3
   widget=Unmapped_Text
   properties {
    text=" Unmapped"
    font {
     bForceMonospaced=false
     bMaterialIsStencil=false
     fontMaterial=null
     fontObject=&Roboto
     letterSpacing=0
     monospacedWidth=1
     size=10.5
     skewAmount=0
     typefaceFontName=Regular
     outlineSettings {
      bApplyOutlineToDropShadows=false
      bMiteredCorners=false
      bSeparateFillAlpha=false
      outlineMaterial=null
      outlineSize=0
      outlineColor=<rgba 0 0 0 1>
     }
    }
   }
   slot {class=HorizontalBoxSlot}
  }
  HorizontalBox_4 {
   widgetClassPath=&HorizontalBox
   parent=WrapBox_0
   widget=HorizontalBox_4
   slot {class=WrapBoxSlot, properties {verticalAlignment=VAlign_Center, padding=<ltrb 10 5 10 5>}}
  }
  DuplicateSwatch_Border {
   widgetClassPath=&Border
   bIsVariable=true
   parent=HorizontalBox_4
   widget=DuplicateSwatch_Border
   properties {desiredSizeScale=<1.25 1>}
   slot {class=HorizontalBoxSlot, properties {padding=<ltrb 0 0 5 0>}}
  }
  DuplicateNumber_Text {
   widgetClassPath=&TextBlock
   bIsVariable=true
   parent=HorizontalBox_4
   widget=DuplicateNumber_Text
   properties {
    text=0
    font {
     bForceMonospaced=false
     bMaterialIsStencil=false
     fontMaterial=null
     fontObject=&Roboto
     letterSpacing=0
     monospacedWidth=1
     size=10.5
     skewAmount=0
     typefaceFontName=Regular
     outlineSettings {
      bApplyOutlineToDropShadows=false
      bMiteredCorners=false
      bSeparateFillAlpha=false
      outlineMaterial=null
      outlineSize=0
      outlineColor=<rgba 0 0 0 1>
     }
    }
   }
   slot {class=HorizontalBoxSlot}
  }
  Duplicate_Text {
   widgetClassPath=&TextBlock
   parent=HorizontalBox_4
   widget=Duplicate_Text
   properties {
    text=" Duplicate"
    font {
     bForceMonospaced=false
     bMaterialIsStencil=false
     fontMaterial=null
     fontObject=&Roboto
     letterSpacing=0
     monospacedWidth=1
     size=10.5
     skewAmount=0
     typefaceFontName=Regular
     outlineSettings {
      bApplyOutlineToDropShadows=false
      bMiteredCorners=false
      bSeparateFillAlpha=false
      outlineMaterial=null
      outlineSize=0
      outlineColor=<rgba 0 0 0 1>
     }
    }
   }
   slot {class=HorizontalBoxSlot}
  }
  Statistics_GridPanel {
   widgetClassPath=&HorizontalBox
   parent=WrapBox_0
   widget=Statistics_GridPanel
   properties {bOverride_Cursor=true}
   slot {class=WrapBoxSlot, properties {verticalAlignment=VAlign_Center, padding=<ltrb 10 5 10 5>}}
  }
  MappedNumber_Text {
   widgetClassPath=&TextBlock
   bIsVariable=true
   parent=Statistics_GridPanel
   widget=MappedNumber_Text
   properties {
    text=0
    font {
     bForceMonospaced=false
     bMaterialIsStencil=false
     fontMaterial=null
     fontObject=&Roboto
     letterSpacing=0
     monospacedWidth=1
     size=10.5
     skewAmount=0
     typefaceFontName=Regular
     outlineSettings {
      bApplyOutlineToDropShadows=false
      bMiteredCorners=false
      bSeparateFillAlpha=false
      outlineMaterial=null
      outlineSize=0
      outlineColor=<rgba 0 0 0 1>
     }
    }
   }
   slot {class=HorizontalBoxSlot}
  }
  Mapped_Text {
   widgetClassPath=&TextBlock
   parent=Statistics_GridPanel
   widget=Mapped_Text
   properties {
    text=" Mapped "
    font {
     bForceMonospaced=false
     bMaterialIsStencil=false
     fontMaterial=null
     fontObject=&Roboto
     letterSpacing=0
     monospacedWidth=1
     size=10.5
     skewAmount=0
     typefaceFontName=Regular
     outlineSettings {
      bApplyOutlineToDropShadows=false
      bMiteredCorners=false
      bSeparateFillAlpha=false
      outlineMaterial=null
      outlineSize=0
      outlineColor=<rgba 0 0 0 1>
     }
    }
   }
   slot {class=HorizontalBoxSlot}
  }
  HorizontalBox_1 {
   widgetClassPath=&HorizontalBox
   parent=WrapBox_0
   widget=HorizontalBox_1
   slot {class=WrapBoxSlot, properties {verticalAlignment=VAlign_Center, padding=<ltrb 10 5 10 5>}}
  }
  OnlyUnresolved_Checkbox {
   widgetClassPath=&EditorUtilityCheckBox
   bIsVariable=true
   parent=HorizontalBox_1
   widget=OnlyUnresolved_Checkbox
   properties {checkedState=Checked}
   slot {class=HorizontalBoxSlot}
  }
  OnlyUnresolved_Text {
   widgetClassPath=&TextBlock
   parent=HorizontalBox_1
   widget=OnlyUnresolved_Text
   properties {
    text=Show Unresolved Only
    font {
     bForceMonospaced=false
     bMaterialIsStencil=false
     fontMaterial=null
     fontObject=&Roboto
     letterSpacing=0
     monospacedWidth=1
     size=10.5
     skewAmount=0
     typefaceFontName=Regular
     outlineSettings {
      bApplyOutlineToDropShadows=false
      bMiteredCorners=false
      bSeparateFillAlpha=false
      outlineMaterial=null
      outlineSize=0
      outlineColor=<rgba 0 0 0 1>
     }
    }
   }
   slot {class=HorizontalBoxSlot}
  }
  HorizontalBox_0 {
   widgetClassPath=&HorizontalBox
   parent=WrapBox_0
   widget=HorizontalBox_0
   slot {class=WrapBoxSlot, properties {verticalAlignment=VAlign_Center, padding=<ltrb 10 5 10 5>}}
  }
  EnableMirroring_Checkbox {
   widgetClassPath=&EditorUtilityCheckBox
   bIsVariable=true
   parent=HorizontalBox_0
   widget=EnableMirroring_Checkbox
   properties {checkedState=Checked}
   slot {class=HorizontalBoxSlot}
  }
  OnlyUnresolved_Text_1 {
   widgetClassPath=&TextBlock
   parent=HorizontalBox_0
   widget=OnlyUnresolved_Text_1
   properties {
    text=Enable Mirror Matching
    font {
     bForceMonospaced=false
     bMaterialIsStencil=false
     fontMaterial=null
     fontObject=&Roboto
     letterSpacing=0
     monospacedWidth=1
     size=10.5
     skewAmount=0
     typefaceFontName=Regular
     outlineSettings {
      bApplyOutlineToDropShadows=false
      bMiteredCorners=false
      bSeparateFillAlpha=false
      outlineMaterial=null
      outlineSize=0
      outlineColor=<rgba 0 0 0 1>
     }
    }
   }
   slot {class=HorizontalBoxSlot}
  }
  Refresh_Button {
   widgetClassPath=&EditorUtilityButton
   parent=WrapBox_0
   widget=Refresh_Button
   properties {toolTipText=Refresh Mappings and Statistics}
   slot {class=WrapBoxSlot, properties {padding=<ltrb 10 5 10 5>}}
  }
  RefreshContents_HorizontalBox {
   widgetClassPath=&HorizontalBox
   parent=Refresh_Button
   widget=RefreshContents_HorizontalBox
   slot {class=ButtonSlot, properties {padding=<ltrb -6 2 -4 2>}}
  }
  Refresh_Image {
   widgetClassPath=&Image
   bIsVariable=true
   parent=RefreshContents_HorizontalBox
   widget=Refresh_Image
   properties {brush {
     bIsDynamicallyLoaded=false
     drawAs=Image
     imageType=FullColor
     mirroring=NoMirror
     resourceName=None
     resourceObject=null
     tiling=NoTile
     imageSize=<20 20>
     margin=<ltrb 0 0 0 0>
     outlineSettings {
      bUseBrushTransparency=false
      roundingType=HalfHeightRadius
      width=0
      color {colorUseRule=UseColor_Specified, specifiedColor=<rgba 0 0 0 0>}
      cornerRadii=<0 0 0 0>
     }
     tintColor {colorUseRule=UseColor_Specified, specifiedColor=<rgba 1 1 1 1>}
     uVRegion {bIsValid=false, max=<0 0>, min=<0 0>}
    }}
   slot {class=HorizontalBoxSlot}
  }
  Refresh {
   widgetClassPath=&TextBlock
   parent=RefreshContents_HorizontalBox
   widget=Refresh
   properties {
    text=Refresh
    font {
     bForceMonospaced=false
     bMaterialIsStencil=false
     fontMaterial=null
     fontObject=&Roboto
     letterSpacing=0
     monospacedWidth=1
     size=10.5
     skewAmount=0
     typefaceFontName=Regular
     outlineSettings {
      bApplyOutlineToDropShadows=false
      bMiteredCorners=false
      bSeparateFillAlpha=false
      outlineMaterial=null
      outlineSize=0
      outlineColor=<rgba 0 0 0 1>
     }
    }
   }
   slot {class=HorizontalBoxSlot, properties {verticalAlignment=VAlign_Center, padding=<ltrb 3 0 0 0>}}
  }
  VerticalBox_69 {widgetClassPath=&VerticalBox, namedSlotHost=ControlMappingExpandable, widget=VerticalBox_69}
  Mapper_SizeBox {
   widgetClassPath=&SizeBox
   bIsVariable=true
   parent=VerticalBox_69
   widget=Mapper_SizeBox
   properties {bOverride_HeightOverride=true, heightOverride=600}
   slot {class=VerticalBoxSlot}
  }
  EditorUtilityScrollBox_1078 {
   widgetClassPath=&EditorUtilityScrollBox
   bIsVariable=true
   parent=Mapper_SizeBox
   widget=EditorUtilityScrollBox_1078
   properties {alwaysShowScrollbar=true}
   slot {class=SizeBoxSlot}
  }
  Source_VerticalBox {
   widgetClassPath=&VerticalBox
   bIsVariable=true
   parent=EditorUtilityScrollBox_1078
   widget=Source_VerticalBox
   properties {bOverride_Cursor=true}
   slot {class=ScrollBoxSlot}
  }
  TextBox {
   widgetClassPath=&HorizontalBox
   bIsVariable=true
   parent=VerticalBox_69
   widget=TextBox
   properties {visibility=Visible}
   slot {
    class=VerticalBoxSlot
    properties {horizontalAlignment=HAlign_Center, verticalAlignment=VAlign_Top, padding=<ltrb 10 10 5 0>, size {sizeRule=Fill, value=1}}
   }
  }
  Image_177 {
   widgetClassPath=&Image
   bIsVariable=true
   parent=TextBox
   widget=Image_177
   properties {brush {
     bIsDynamicallyLoaded=false
     drawAs=Image
     imageType=FullColor
     mirroring=NoMirror
     resourceName=None
     resourceObject=&/T_UI_IconLibrary_checkmark_64
     tiling=NoTile
     imageSize=<30 24>
     margin=<ltrb 0 0 0 0>
     outlineSettings {
      bUseBrushTransparency=false
      roundingType=HalfHeightRadius
      width=0
      color {colorUseRule=UseColor_Specified, specifiedColor=<rgba 0 0 0 0>}
      cornerRadii=<0 0 0 0>
     }
     tintColor {colorUseRule=UseColor_Specified, specifiedColor=<rgba 1 1 1 1>}
     uVRegion {bIsValid=false, max=<0 0>, min=<0 0>}
    }}
   slot {class=HorizontalBoxSlot, properties {horizontalAlignment=HAlign_Center, verticalAlignment=VAlign_Center}}
  }
  Complete_Text {
   widgetClassPath=&TextBlock
   bIsVariable=true
   parent=TextBox
   widget=Complete_Text
   properties {
    justification=Center
    text="All controls have been mapped!"
    font {
     bForceMonospaced=false
     bMaterialIsStencil=false
     fontMaterial=null
     fontObject=&Roboto
     letterSpacing=0
     monospacedWidth=1
     size=12
     skewAmount=0
     typefaceFontName=Regular
     outlineSettings {
      bApplyOutlineToDropShadows=false
      bMiteredCorners=false
      bSeparateFillAlpha=false
      outlineMaterial=null
      outlineSize=0
      outlineColor=<rgba 0 0 0 1>
     }
    }
   }
   slot {class=HorizontalBoxSlot, properties {horizontalAlignment=HAlign_Center, verticalAlignment=VAlign_Center}}
  }
 ]
}
