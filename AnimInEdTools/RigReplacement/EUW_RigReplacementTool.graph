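# Editor utility widget tool that replaces a Source Control Rig with a Target rig on a selected Level Sequence, carrying animation over and relinking Anim Sequences.
# It validates rigs and the Animatable Creature asset, checks out files, then spawns the target actor with its rig track.
# Embedded Python copies animation through AnimLibrary exporters, pastes it using the mapping widget, rewires attach tracks and prunes empty anim layers.
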EditorUtilityWidgetBlueprint /AnimInEdTools/RigReplacement/EUW_RigReplacementTool
graphs [
 "Adding AnimSeq to Map, Query AnimSeq State" {nodes [
   "Adding AnimSeq to Map, Query AnimSeq State"() {is=title K2Node_FunctionEntry}
   > n1 = Sequence()
   > n2 = GetBindings!(
    execute
    Sequence:Movie Scene Sequence Object Reference=Get Selected Level Sequence() -> (Selected Level Sequence:Level Sequence Object Reference)
   ) -> (then, ReturnValue:Array of Movie Scene Binding Proxy Structures)
   > n3 = GetAnimSequenceLinkFromLevelSequence!(execute, InLevelSequence:Level Sequence Object Reference=Get Selected Level Sequence()) -> (then, ReturnValue:Level Sequence Anim Sequence Link Object Reference)
   > n4 = For Each Loop(Array:Array of Movie Scene Binding Proxy Structures=@n2) -> (Array Element:Movie Scene Binding Proxy Structure)
   > Set Anim_Movie_Binding(Anim_Movie_Binding:Movie Scene Binding Proxy Structure=@n4) -> (Output_Get:Movie Scene Binding Proxy Structure)
   > n5 = For Each Loop!(
    Exec
    Array:Array of Level Sequence Anim Sequence Link Item Structures=Get AnimSequenceLinks!(self:Level Sequence Anim Sequence Link Object Reference=@n3) -> (AnimSequenceLinks:Array of Level Sequence Anim Sequence Link Item Structures)
   ) -> (
    LoopBody
    Array Element_SkelTrackGuid:Guid Structure
    Array Element_PathToAnimSequence:Soft Object Path Structure
    Array Element_bAutoBake:bool=~
    Array Element_bExportTransforms:bool=~
    Array Element_bExportMorphTargets:bool=~
    Array Element_bExportAttributeCurves:bool=~
    Array Element_bExportMaterialCurves:bool=~
    Array Element_bSkipCurvesWithZeroValue:bool=~
    Array Element_bSetRetargetSourceAsset:bool=~
    Array Element_Interpolation:EAnimInterpolationType Enum
    Array Element_CurveInterpolation:ERichCurveInterpMode Enum
    Array Element_bRecordInWorldSpace:bool=~
    Array Element_bEvaluateAllSkeletalMeshComponents:bool=~
    Array Element_IncludeAnimationNames:Array of Strings
    Array Element_ExcludeAnimationNames:Array of Strings
    Array Element_bRemoveExcludedCurves:bool=~
    Array Element_WarmUpFrames:Frame Number Structure
    Array Element_DelayBeforeStart:Frame Number Structure
    Array Element_bUseCustomTimeRange:bool=~
    Array Element_CustomStartFrame:Frame Number Structure
    Array Element_CustomEndFrame:Frame Number Structure
    Array Element_CustomDisplayRate:Frame Rate Structure
    Array Element_bUseCustomFrameRate:bool=~
    Array Element_CustomFrameRate:Frame Rate Structure
    Array Index
    Completed
   )
   > n6 = FindBindingById!(execute, Sequence:Movie Scene Sequence Object Reference=Get Selected Level Sequence(), BindingId:Guid Structure=@n5) -> (then, ReturnValue_BindingID:Guid Structure, ReturnValue_Sequence:Movie Scene Sequence Object Reference)
   n7 = Get Anim_Movie_Binding() -> (Anim_Movie_Binding:Movie Scene Binding Proxy Structure)
   > Branch(Condition="Equal (Guid)"(
     A:"Guid Structure (by ref)"=@n6
     B:"Guid Structure (by ref)"=Break Movie Scene Binding Proxy!(MovieSceneBindingProxy:"Movie Scene Binding Proxy Structure (by ref)"=@n7) -> (BindingID:Guid Structure, Sequence:Movie Scene Sequence Object Reference) {is=title K2Node_BreakStruct}
    ) -> (ReturnValue:bool=~) {is=title K2Node_PromotableOperator})
   > n8 = LoadAsset(AssetPath=BreakSoftObjectPath(InSoftObjectPath=@n5.Array Element_PathToAnimSequence))
   > n9 = Cast To AnimSequence!(execute, Object=@n8) -> (then, CastFailed, AsAnim Sequence:Anim Sequence Object Reference)
   > n10 = Query Anim Sequence!(
    execute
    self
    InFile:String=GetSystemPath!(Object:Object Reference=@n9) -> (ReturnValue:String)
    Anim Sequence:Anim Sequence Object Reference=@n9
   ) -> (then, Was it done:bool)
   > Branch(Condition=@n10)
   > CheckoutLoadedAsset!(execute, AssetToCheckout:Object Reference=@n9) -> (then, ReturnValue:bool)
   > n11 = Does Anim Modifier Exists!(
    execute
    self
    incoming AnimSeq:Interface Asset User Data Interface=@n9
    incoming LS:Level Sequence Object Reference=Get Selected Level Sequence()
   ) -> (then, Selected LS missing Modifier:Level Sequence Object Reference)
   > Add Unique(
    TargetArray:Array of Level Sequence Object References=Get Selected LS Missing Modifier array() -> (Selected LS Missing Modifier array:Array of Level Sequence Object References)
    NewItem:"Level Sequence Object Reference (by ref)"=@n11
   )
   Return Node(execute=@n1.then_1)
  ]}
 Are CRs Valid {nodes [
   Are CRs Valid() {is=title K2Node_FunctionEntry}
   n1 = Get MappingWidget() -> (MappingWidget:EUW Mappings Widget Object Reference)
   > n2 = Is Valid(InputObject=Get SourceRig!(self:EUW Mappings Widget Object Reference=@n1) -> (SourceRig:Control Rig Blueprint Object Reference))
   Show Message Dialog(
    execute=@n2.Is Not Valid
    Title="To Text (String)"(InString="Source Control Rig  not valid")
    Message="To Text (String)"(InString=Please select a valid Source CR)
    MessageType=Ok
    DefaultValue=Ok
    MessageCategory=Error
   )
   > Return Node!(execute, Control Rigs are Valid:bool=false)
   n3 = Is Valid(
    exec=@n2.Is Valid
    InputObject=Get TargetRig!(self:EUW Mappings Widget Object Reference=@n1) -> (TargetRig:Control Rig Blueprint Object Reference)
   )
   Show Message Dialog(
    execute=@n3.Is Not Valid
    Title="To Text (String)"(InString=New Asset Error)
    Message="To Text (String)"(InString=Please select a valid AC file in the New Asset slot)
    MessageType=Ok
    DefaultValue=Ok
    MessageCategory=Error
   )
   > Return Node(Control Rigs are Valid=false)
   Return Node(execute=@n3.Is Valid, Control Rigs are Valid=true)
  ]}
 Are Control rigs and AC files valid {nodes [
   Are Control rigs and AC files valid() {is=title K2Node_FunctionEntry}
   > n1 = Is Target AC file Valid!(execute, self) -> (then, Is Valid:bool)
   > Branch(Condition=@n1)
   > n2 = Are CRs Valid!(execute, self) -> (then, Control Rigs are Valid:bool) {is=title K2Node_CallFunction}
   > Return Node!(execute, Is valid?:bool=@n2)
  ]}
 Bake to Anim Sequence {nodes [
   n1 = Bake to Anim Sequence!() -> (then, Pasted Level Sequence:Level Sequence Object Reference) {is=title K2Node_FunctionEntry}
   > n2 = GetBindings(Sequence=@n1)
   > n3 = GetAnimSequenceLinkFromLevelSequence(InLevelSequence=Get Pasted Level Sequence() -> (Pasted Level Sequence:Level Sequence Object Reference))
   > n4 = For Each Loop!(Exec, Array:Array of Movie Scene Binding Proxy Structures=@n2) -> (LoopBody, Array Element:Movie Scene Binding Proxy Structure, Array Index, Completed)
   > Set Anim Binding(Anim Binding:Movie Scene Binding Proxy Structure=@n4) -> (Output_Get:Movie Scene Binding Proxy Structure)
   > n5 = For Each Loop!(Exec, Array:Array of Level Sequence Anim Sequence Link Item Structures=Get AnimSequenceLinks(self=@n3)) -> (
    LoopBody
    Array Element_SkelTrackGuid:Guid Structure
    Array Element_PathToAnimSequence:Soft Object Path Structure
    Array Element_bAutoBake:bool=~
    Array Element_bExportTransforms:bool=~
    Array Element_bExportMorphTargets:bool=~
    Array Element_bExportAttributeCurves:bool=~
    Array Element_bExportMaterialCurves:bool=~
    Array Element_bSkipCurvesWithZeroValue:bool=~
    Array Element_bSetRetargetSourceAsset:bool=~
    Array Element_Interpolation:EAnimInterpolationType Enum
    Array Element_CurveInterpolation:ERichCurveInterpMode Enum
    Array Element_bRecordInWorldSpace:bool=~
    Array Element_bEvaluateAllSkeletalMeshComponents:bool=~
    Array Element_IncludeAnimationNames:Array of Strings
    Array Element_ExcludeAnimationNames:Array of Strings
    Array Element_bRemoveExcludedCurves:bool=~
    Array Element_WarmUpFrames:Frame Number Structure
    Array Element_DelayBeforeStart:Frame Number Structure
    Array Element_bUseCustomTimeRange:bool=~
    Array Element_CustomStartFrame:Frame Number Structure
    Array Element_CustomEndFrame:Frame Number Structure
    Array Element_CustomDisplayRate:Frame Rate Structure
    Array Element_bUseCustomFrameRate:bool=~
    Array Element_CustomFrameRate:Frame Rate Structure
    Array Index
    Completed
   )
   > n6 = FindBindingById(Sequence=Get Pasted Level Sequence(), BindingId=@n5)
   n7 = Get Anim Binding() -> (Anim Binding:Movie Scene Binding Proxy Structure)
   > Branch(Condition="Equal (Guid)"(A=@n6, B=Break Movie Scene Binding Proxy(MovieSceneBindingProxy=@n7)) -> (ReturnValue=~))
   > Set StartFrame(StartFrame:Frame Number Structure=@n5.Array Element_CustomStartFrame) -> (Output_Get:Frame Number Structure)
   > Set End Frame(End Frame:Frame Number Structure=@n5.Array Element_CustomEndFrame) -> (Output_Get:Frame Number Structure)
   > Set CustomTimeFrame(CustomTimeFrame:bool=@n5.Array Element_bUseCustomTimeRange) -> (Output_Get:bool)
   > n8 = LoadAsset(AssetPath=BreakSoftObjectPath(InSoftObjectPath=@n5.Array Element_PathToAnimSequence))
   > n9 = Cast To AnimSequence(Object=@n8)
   > Set Anim Sequence(Anim Sequence:Anim Sequence Object Reference=@n9) -> (Output_Get:Anim Sequence Object Reference)
   > Set CustomDisplayRate(CustomDisplayRate:Frame Rate Structure=@n5.Array Element_CustomDisplayRate) -> (Output_Get:Frame Rate Structure)
   > Execute Python Script!(
    execute
    PythonScript="import unreal\r\n\r\n## Get the asset tools\r\nasset_tools = unreal.AssetToolsHelpers.get_asset_tools()\r\n\r\n## Grab the Level Editor World\r\neditor_subsystem = unreal.get_editor_subsystem(unreal.UnrealEditorSubsystem)\r\nworld = editor_subsystem.get_editor_world()\r\n\r\nanim_seq_export_options = unreal.AnimSeqExportOption()\r\nanim_seq_export_options.export_transforms = True\r\nanim_seq_export_options.export_morph_targets = True\r\nanim_seq_export_options.custom_display_rate = cdr\r\nanim_seq_export_options.use_custom_time_range = custom_time\r\nanim_seq_export_options.custom_start_frame = start_frame\r\nanim_seq_export_options.custom_end_frame = end_frame\r\n\r\n## Bake to the created AnimSequence\r\n##unreal.SequencerTools.export_anim_sequence(world, selected_ls, anim, anim_seq_export_options, binding, False)\r\nunreal.ControlRigSequencer.export_anim_sequence_from_sequencer(anim,anim_seq_export_options, binding, False)"
    Selected_LS:Level Sequence Object Reference=Get Pasted Level Sequence()
    binding:Movie Scene Binding Proxy Structure=Get Anim Binding()
    anim:Anim Sequence Object Reference=Get Anim Sequence() -> (Anim Sequence:Anim Sequence Object Reference)
    custom_time:bool=Get CustomTimeFrame() -> (CustomTimeFrame:bool)
    start_frame:Frame Number Structure=Get StartFrame() -> (StartFrame:Frame Number Structure)
    end_frame:Frame Number Structure=Get End Frame() -> (End Frame:Frame Number Structure)
    cdr:Frame Rate Structure=Get CustomDisplayRate() -> (CustomDisplayRate:Frame Rate Structure)
   ) -> (then, ReturnValue) {inputs [Selected_LS,binding,anim,custom_time,start_frame,end_frame,cdr]}
   > Apply All Animimation Modifiers!(execute, AnimSequence:Anim Sequence Object Reference=Get Anim Sequence()) -> (then)
   > SaveLoadedAssets(
    AssetsToSave=Make Array!("[0]":Object Reference=Get Anim Sequence(), "[1]":Object Reference=Get Pasted Level Sequence()) -> (Array:Array of Object References)
    bOnlyIfIsDirty=false
   )
   > CloseLevelSequence!(execute) -> (then)
   > SaveLoadedAsset(AssetToSave=Get Pasted Level Sequence(), bOnlyIfIsDirty=true)
   > Return Node!(execute)
  ]}
 Checked out by another user warning {nodes [
   n1 = Checked out by another user warning!() -> (then, Checked out by:String, Dipslay aset name:String) {is=title K2Node_FunctionEntry}
   > Show Message Dialog(
    Title=MakeLiteralText(Value="LOCTEXT(\"Cannot Edit File\")")
    Message="To Text (String)"(InString=Append!(A="The seleted file  ", B=@n1.Dipslay aset name, C:String=" is currently checked out by ", D:String=@n1) -> (ReturnValue))
    MessageType=Ok
    DefaultValue=Ok
    MessageCategory=Error
   )
   > Return Node!(execute, Was it done:bool=false)
  ]}
 Continue After Source Selected Binding {nodes [
   n1 = Continue After Source Selected Binding!() -> (then, selected source binding:Movie Scene Binding Proxy Structure) {is=title K2Node_FunctionEntry}
   > n2 = Set source_binding(source_binding:Movie Scene Binding Proxy Structure=@n1) -> (Output_Get:Movie Scene Binding Proxy Structure)
   > SelectBindings!(
    execute
    ObjectBindings:Array of Movie Scene Binding Proxy Structures=Make Array!("[0]":Movie Scene Binding Proxy Structure=@n2) -> (Array:Array of Movie Scene Binding Proxy Structures)
   ) -> (then)
   > Copy Animation()
   > Paste Anim()
   > Link to Anim Sequence()
   > Show Message Dialog(
    Title="To Text (String)"(InString=Rig Replacement Success)
    Message="To Text (String)"(InString=The selected Level Sequences have been updated to the Target Rig)
    MessageType=Ok
    DefaultValue=Ok
    MessageCategory=Success
   )
   > Return Node!(execute)
  ]}
 Copy Animation {nodes [
   Copy Animation() {is=title K2Node_FunctionEntry}
   > n1 = For Each Loop with Break!(
    Exec:Exec
    Array:Array of Movie Scene Binding Proxy Structures=GetSelectedBindings!() -> (ReturnValue:Array of Movie Scene Binding Proxy Structures)
    Break:Exec=@n4.then
   ) -> (LoopBody:Exec, Array Element:Movie Scene Binding Proxy Structure, Array Index:int, Completed:Exec)
   n2 = Get source_binding() -> (source_binding:Movie Scene Binding Proxy Structure)
   > Branch(Condition=AND Boolean(
     A=IsValid!(InBinding:"Movie Scene Binding Proxy Structure (by ref)"=@n1) -> (ReturnValue)
     B="Equal (MovieSceneBindingProxy)"(A:Movie Scene Binding Proxy Structure=@n1, B:Movie Scene Binding Proxy Structure=@n2) -> (ReturnValue:bool=~) {is=title K2Node_PromotableOperator}
    ))
   > n3 = Execute Python Script!(
    execute
    PythonScript="import unreal\r\nfrom AnimLibrary.core.libanim.base.pipeline import OperatorPipeline\r\nfrom AnimLibrary.core.libanim.exporters.controlrig_exporter import ControlRigExporter\r\nfrom AnimLibrary.core.libanim.exporters.worldspace_exporter import WorldSpaceExporter\r\nfrom AnimLibrary.core.libanim.exporters.constraint_exporter import ConstraintExporter\r\nfrom AnimLibrary.core.libanim.context import SelectionContext\r\nfrom AnimLibrary.core.constants import AnimOperation\r\n\r\ncontext = SelectionContext.from_current_selection(\r\n    operation=AnimOperation.COPY\r\n)\r\npipeline = OperatorPipeline(context, [ControlRigExporter, WorldSpaceExporter, ConstraintExporter])\r\npipeline.execute()\r\nanim_data = pipeline.aggregate()\r\nexport_data = unreal.create_python_object_handle(anim_data)"
   ) -> (then, ReturnValue, export_data:Python Object Handle Object Reference) {outputs [export_data]}
   > Set export_data(export_data:Python Object Handle Object Reference=@n3.export_data) -> (Output_Get:Python Object Handle Object Reference)
   > EmptySelection!(execute) -> (then)
   > RefreshCurrentLevelSequence!(execute) -> (then)
   > Set Current Time!(
    execute
    PlaybackParams_Frame_FrameNumber_Value:int=int - int(A=Get Start Time() -> (Start Time:int=0), B=-1)
    PlaybackParams_Frame_SubFrame:float=0
    PlaybackParams_Time:float=0
    PlaybackParams_MarkedFrame:String
    PlaybackParams_Timecode:Timecode Structure=(Hours=0,Minutes=0,Seconds=0,Frames=0,Subframe=0,bDropFrameFormat=False)
    PlaybackParams_PositionType:EMovieScenePositionType Enum=Frame
    PlaybackParams_UpdateMethod:EUpdatePositionMethod Enum=Play
    PlaybackParams_bHasJumped:bool=False
    TimeUnit:EMovieSceneTimeUnit Enum=DisplayRate
   ) -> (then) {is=title K2Node_CallFunction}
   > n4 = Set Current Time(
    PlaybackParams_Frame_FrameNumber_Value=Get Start Time() -> (Start Time=0)
    PlaybackParams_Frame_SubFrame=0
    PlaybackParams_Time=0
    PlaybackParams_Timecode=(Hours=0,Minutes=0,Seconds=0,Frames=0,Subframe=0,bDropFrameFormat=False)
    PlaybackParams_PositionType=Frame
    PlaybackParams_UpdateMethod=Play
    PlaybackParams_bHasJumped=False
    TimeUnit=DisplayRate
   )
  ]}
 Copy Paste Function {nodes [
   n1 = Copy Paste Function() {is=title K2Node_FunctionEntry}
   n2 = GetFocusedLevelSequence!(execute=[@n1.then, @n3.Is Not Valid]) -> (then, ReturnValue:Level Sequence Object Reference)
   > n3 = Is Valid(InputObject=@n2)
   > n4 = Set CachedLevelSequence(CachedLevelSequence:Level Sequence Object Reference=@n2) -> (Output_Get:Level Sequence Object Reference)
   > n5 = GetPlaybackStart!(execute, Sequence:Movie Scene Sequence Object Reference=@n4) -> (then, ReturnValue:int=0)
   > Set Start Time(Start Time:int=@n5) -> (Output_Get:int=0)
   > n6 = GetControlRigs!(execute, LevelSequence:Level Sequence Object Reference=Get CachedLevelSequence() -> (CachedLevelSequence:Level Sequence Object Reference)) -> (then, ReturnValue:Array of Control Rig Sequencer Binding Proxy Structures)
   > n7 = For Each Loop with Break!(Exec, Array:Array of Control Rig Sequencer Binding Proxy Structures=@n6, Break) -> (
    LoopBody
    Array Element_Proxy:Movie Scene Binding Proxy Structure
    Array Element_ControlRig:Control Rig Object Reference
    Array Element_Track:Movie Scene Control Rig Parameter Track Object Reference
    Array Index
    Completed
   )
   n8 = Branch(execute=@n7.Completed, Condition="integer > integer"(A=Get multiple source channels() -> (multiple source channels:int=0), B=1))
   > n9 = Create Widget!(
    execute:Exec
    WorldContextObject:Object Reference
    Class:User Widget Class Reference=&/EUW_SelectBinding.EUW_SelectBinding_C
    OwningPlayer:Player Controller Object Reference
   ) -> (then:Exec, ReturnValue:EUW Select Binding Object Reference)
   > n10 = Set SelectedBinding(SelectedBinding:EUW Select Binding Object Reference=@n9) -> (Output_Get:EUW Select Binding Object Reference)
   > Set SourceRig!(
    execute
    SourceRig:Control Rig Blueprint Object Reference=Get SourceRig(self=Get MappingWidget())
    self:EUW Select Binding Object Reference=@n10
   ) -> (then, Output_Get:Control Rig Blueprint Object Reference)
   > AddChild(
    self=Get ModalLayer() -> (ModalLayer:Named Slot Object Reference)
    Content=Get SelectedBinding() -> (SelectedBinding:EUW Select Binding Object Reference)
   )
   > Bind Event to ED Binding Select(self:EUW Select Binding Object Reference=Get SelectedBinding(), Delegate=Create Event()) {is=title K2Node_AddDelegate}
   n11 = GetControlRigs(execute=@n8.else, LevelSequence=Get CachedLevelSequence())
   > n12 = For Each Loop with Break(Array=@n11, Break=@n18.then)
   n13 = GetBindings(execute=@n12.Completed, Sequence=Get CachedLevelSequence())
   > n14 = For Each Loop!(Exec, Array:Array of Movie Scene Binding Proxy Structures=@n13) -> (LoopBody, Array Element:Movie Scene Binding Proxy Structure, Array Index, Completed)
   Continue After Source Selected Binding!(execute=@n14.Completed, self, selected source binding:Movie Scene Binding Proxy Structure=Get source_binding()) -> (then) {is=title K2Node_CallFunction}
   n15 = GetTracks!(execute=@n14.LoopBody, InBinding:"Movie Scene Binding Proxy Structure (by ref)"=@n14) -> (then, ReturnValue:Array of Movie Scene Track Object References)
   > n16 = For Each Loop(Array:Array of Movie Scene Track Object References=@n15) -> (Array Element:Movie Scene Track Object Reference)
   > Branch(Condition="Equal (Object)"(A=@n16, B=Get temp_control_rig_track() -> (temp_control_rig_track:Movie Scene Control Rig Parameter Track Object Reference)))
   > Set source_binding(source_binding=@n14)
   n17 = Get SourceRig(self=Get MappingWidget())
   Branch(
    execute=@n12.LoopBody
    Condition="Equal Exactly (String)"(A=GetDisplayName(Object=@n12.Array Element_ControlRig), B=GetDisplayName(Object=@n17))
   )
   > n18 = Set temp_control_rig_track(temp_control_rig_track:Movie Scene Control Rig Parameter Track Object Reference=@n12.Array Element_Track) -> (Output_Get:Movie Scene Control Rig Parameter Track Object Reference)
   n19 = Get SourceRig(self=Get MappingWidget())
   Branch(
    execute=@n7.LoopBody
    Condition="Equal Exactly (String)"(A=GetDisplayName(Object=@n7.Array Element_ControlRig), B=GetDisplayName(Object=@n19))
   )
   > Set multiple source channels(multiple source channels:int=int + int(A=1, B=Get multiple source channels() -> (multiple source channels=0))) -> (Output_Get:int=0)
  ]}
 Dialog Box {nodes [
   Dialog Box() {is=title K2Node_FunctionEntry}
   n1 = Get MappingWidget()
   n2 = GetDisplayName(Object=Get TargetRig(self=@n1))
   n3 = GetDisplayName(Object=Get SourceRig(self=@n1))
   > n4 = Show Message Dialog(
    Title="To Text (String)"(InString=Are you sure you want to continue?)
    Message="To Text (String)"(InString=Append!(
      A="You are about to copy animation from "
      B=@n3
      C=" on to "
      D=@n2
      E:String=.
      F:String
      G:String="\r\n\r\nAre you sure you want to do this?"
     ) -> (ReturnValue))
    MessageType=YesNo
    DefaultValue=No
    MessageCategory=Warning
   )
   > Return Node!(execute, ReturnValue:EAppReturnType Enum=@n4)
  ]}
 Does Anim Modifier Exists {nodes [
   n1 = Does Anim Modifier Exists!() -> (then, incoming AnimSeq:Interface Asset User Data Interface, incoming LS:Level Sequence Object Reference) {is=title K2Node_FunctionEntry}
   > n2 = GetAssetUserDataOfClass!(
    execute
    self:Interface Asset User Data Interface=@n1
    InUserDataClass:Asset User Data Class Reference=/Script/AnimationModifiers.AnimationModifiersAssetUserData
   ) -> (then, ReturnValue:Animation Modifiers Asset User Data Object Reference)
   > n3 = Branch(Condition=Is Empty!(TargetArray:Array of Animation Modifier Object References=Get AnimationModifierInstances!(self:Animation Modifiers Asset User Data Object Reference=@n2) -> (AnimationModifierInstances:Array of Animation Modifier Object References)) -> (ReturnValue:bool))
   > Return Node!(execute, Selected LS missing Modifier:Level Sequence Object Reference=Get incoming LS() -> (incoming LS:Level Sequence Object Reference))
   Return Node(execute=@n3.else)
  ]}
 EventGraph {nodes [
   "On Clicked (Button_?)"() {is=title K2Node_ComponentBoundEvent}
   > Execute Python Script!(
    execute
    PythonScript="import webbrowser\r\n\r\nwebbrowser.open(\"https://epicgames.atlassian.net/wiki/spaces/FORT/pages/314444983/Rig+Replacement+Tool\")"
   ) -> (then, ReturnValue)
   "On Clicked (Copy_Anim_Btn)"() {is=title K2Node_ComponentBoundEvent}
   n1 = Get Selected Level Sequence()
   > n2 = Query LevelSequence!(execute, self, InFile:String=GetSystemPath(Object=@n1), CurrentLS:Level Sequence Object Reference=@n1) -> (then, Was it done:bool)
   > Branch(Condition=@n2)
   > n3 = Are Control rigs and AC files valid!(execute, self) -> (then, Is valid?:bool) {is=title K2Node_CallFunction}
   > Branch(Condition=@n3)
   > n4 = Validation Of controlrigs in LevelSequence!(execute, self) -> (then, Is Valid:bool)
   > n5 = Branch(Condition=@n4)
   n6 = GetDisplayName(Object=Get Selected Level Sequence())
   n7 = Get MappingWidget()
   n8 = GetDisplayName(Object=Get SourceRig(self=@n7))
   n9 = GetDisplayName(Object=Get TargetRig(self=@n7))
   > n10 = Show Message Dialog(
    Title=Get Rig Replacement Tool Title() -> (Rig Replacement Tool Title:Text)
    Message="To Text (String)"(InString=Append!(
      A="Are you sure you want to replace the \r\nSource Rig: \r\n"
      B=@n8
      C=" \r\nwith the Target Rig: \r\n"
      D=@n9
      E=" \r\non the LevelSequence: \r\n"
      F=@n6
     ) -> (ReturnValue))
    MessageType=YesNo
    DefaultValue=No
    MessageCategory=Info
   )
   > Branch(Condition="Equal (Enum)"(A:EAppReturnType Enum=@n10, B:EAppReturnType Enum=Yes))
   > "Adding AnimSeq to Map, Query AnimSeq State"() {is=title K2Node_CallFunction}
   > n11 = Sequence()
   > CheckoutLoadedAsset(AssetToCheckout=Get Selected Level Sequence())
   Import Target Control Rig into LevelSequence!(execute=@n11.then_1, self, level_sequence:Level Sequence Object Reference=Get Selected Level Sequence()) -> (then)
   > Copy Paste Function() {is=title K2Node_CallFunction}
   PopupDialog(execute=@n5.else)
   n12 = Asset Has Changed!() -> (OutputDelegate, then, NewAsset:String, Current Project:BP Animation Browser Project Object Reference) {is=title K2Node_CustomEvent}
   > n13 = Populate Source Rig!(execute, self, B:String=@n12.NewAsset, Project:BP Animation Browser Project Object Reference=@n12.Current Project) -> (then, MappingWidget:EUW Mappings Widget Object Reference)
   > Target or Source Rig Updated(self:EUW Mappings Widget Object Reference=@n13)
   Event Construct()
   > n14 = Sequence()
   > Populate Slots()
   > ValidationLevelSequence()
   > Set ScrollBoxHeight(ScrollBoxHeight:double=Get HeightOverride!(self:Size Box Object Reference=Get SizeBox_Mapper() -> (SizeBox_Mapper:Size Box Object Reference)) -> (HeightOverride:float)) -> (Output_Get:double)
   Bind Event to ED Asset Has Changed(
    execute=@n14.then_1
    self:EUW Level Sequence Browser Object Reference=Resolve Soft Reference!(Input:EUW Level Sequence Browser Soft Object Reference=Get LevelSequenceBrowser() -> (LevelSequenceBrowser:EUW Level Sequence Browser Soft Object Reference)) -> (Output:EUW Level Sequence Browser Object Reference)
    Delegate=@n12
   ) {is=title K2Node_AddDelegate}
   n15 = Event OnMouseEnter()
   ValidationLevelSequence(execute=[@n15.then, @n16.then])
   > Update Level Sequence Name()
   n16 = Event OnMouseLeave!() -> (OutputDelegate, then, MouseEvent:"Pointer Event Structure (by ref)") {is=title K2Node_Event}
  ]}
 Import Target Control Rig into LevelSequence {nodes [
   n1 = Import Target Control Rig into LevelSequence!() -> (then, level_sequence:Level Sequence Object Reference) {is=title K2Node_FunctionEntry}
   > OpenLevelSequence!(execute, LevelSequence:Level Sequence Object Reference=@n1) -> (then, ReturnValue:bool)
   n2 = Get Target_AC!(self:EUW Mappings Widget Object Reference=Get MappingWidget()) -> (Target_AC:Animatable Creature Soft Object Reference)
   > n3 = LoadClassAsset_Blocking(AssetClass=Get Sequencer Actor!(self:Animatable Creature Object Reference=Resolve Soft Reference(Input:Animatable Creature Soft Object Reference=@n2) -> (Output:Animatable Creature Object Reference)) -> (Sequencer Actor:Object Soft Class Reference))
   > n4 = CastToControlRigBlueprint!(execute, Object:Object Reference=Get TargetRig(self=Get MappingWidget())) -> (CastSucceeded:Exec, CastFailed:Exec, AsControlRigBlueprint:Control Rig Blueprint Object Reference)
   > n5 = GeneratedClass!(execute, BlueprintObj:Blueprint Object Reference=@n4) -> (then, ReturnValue:Object Class Reference)
   > n6 = Execute Python Script!(
    execute
    PythonScript="\r\nimport unreal\r\n\r\n##create control rig class\r\n##control_rig_class = target_control_rig.get_control_rig_class()\r\n\r\n## Load the Actor Blueprint asset and get its generated class\r\n##blueprint_asset = unreal.EditorAssetLibrary.load_blueprint_class(actor_blueprint_path)\r\n\r\n## Get the Level Sequence Editor Subsystem\r\nls_subsys = unreal.get_editor_subsystem(unreal.LevelSequenceEditorSubsystem)\r\n\r\n## Add the spawnable from the class\r\nspawnable_binding = ls_subsys.add_spawnable_from_class(sequence, blueprint_asset)\r\n\r\ntarget_binding_name = spawnable_binding.get_name()\r\n\r\ncontrol_rig_track = unreal.ControlRigSequencerLibrary.find_or_create_control_rig_track(\r\n    world = unreal.EditorLevelLibrary.get_editor_world(),\r\n    level_sequence=sequence,\r\n    binding=spawnable_binding,\r\n    control_rig_class=control_rig_class\r\n)"
    sequence:Level Sequence Object Reference=Get level_sequence() -> (level_sequence:Level Sequence Object Reference)
    control_rig_class:Object Class Reference=@n5
    blueprint_asset:Object Class Reference=@n3
   ) -> (then, ReturnValue, target_binding_name:String) {inputs [sequence, control_rig_class, blueprint_asset], outputs [target_binding_name]}
   > Set target_sequence_actor(target_sequence_actor:String=@n6.target_binding_name) -> (Output_Get:String)
   > RefreshCurrentLevelSequence()
   > Return Node!(execute)
  ]}
 Is Target AC file Valid {nodes [
   Is Target AC file Valid() {is=title K2Node_FunctionEntry}
   n1 = Get MappingWidget()
   > n2 = Branch(Condition=IsValidSoftObjectReference(SoftObjectReference=Get Target_AC(self=@n1)))
   > Return Node!(execute, Is Valid:bool=true)
   Show Message Dialog(
    execute=@n2.else
    Title="To Text (String)"(InString="Error!")
    Message="To Text (String)"(InString=Please select a valid Target AC file)
    MessageType=Ok
    DefaultValue=Ok
    MessageCategory=Error
   )
   > Return Node(Is Valid=false)
  ]}
 Link to Anim Sequence {nodes [
   Link to Anim Sequence() {is=title K2Node_FunctionEntry}
   > n1 = Sequence()
   > Clear(TargetArray:Array of Anim Sequence Object References=Get AnimSeq_Array() -> (AnimSeq_Array:Array of Anim Sequence Object References))
   > n2 = GetBindings(Sequence=Get CachedLevelSequence())
   > n3 = GetAnimSequenceLinkFromLevelSequence(InLevelSequence=Get CachedLevelSequence())
   > n4 = For Each Loop(Array:Array of Movie Scene Binding Proxy Structures=@n2) -> (Array Element:Movie Scene Binding Proxy Structure)
   > n5 = For Each Loop!(Exec, Array:Array of Level Sequence Anim Sequence Link Item Structures=Get AnimSequenceLinks(self=@n3)) -> (
    LoopBody
    Array Element_SkelTrackGuid:Guid Structure
    Array Element_PathToAnimSequence:Soft Object Path Structure
    Array Element_bAutoBake:bool=~
    Array Element_bExportTransforms:bool=~
    Array Element_bExportMorphTargets:bool=~
    Array Element_bExportAttributeCurves:bool=~
    Array Element_bExportMaterialCurves:bool=~
    Array Element_bSkipCurvesWithZeroValue:bool=~
    Array Element_bSetRetargetSourceAsset:bool=~
    Array Element_Interpolation:EAnimInterpolationType Enum
    Array Element_CurveInterpolation:ERichCurveInterpMode Enum
    Array Element_bRecordInWorldSpace:bool=~
    Array Element_bEvaluateAllSkeletalMeshComponents:bool=~
    Array Element_IncludeAnimationNames:Array of Strings
    Array Element_ExcludeAnimationNames:Array of Strings
    Array Element_bRemoveExcludedCurves:bool=~
    Array Element_WarmUpFrames:Frame Number Structure
    Array Element_DelayBeforeStart:Frame Number Structure
    Array Element_bUseCustomTimeRange:bool=~
    Array Element_CustomStartFrame:Frame Number Structure
    Array Element_CustomEndFrame:Frame Number Structure
    Array Element_CustomDisplayRate:Frame Rate Structure
    Array Element_bUseCustomFrameRate:bool=~
    Array Element_CustomFrameRate:Frame Rate Structure
    Array Index
    Completed
   )
   > n6 = FindBindingById(Sequence=Get CachedLevelSequence(), BindingId=@n5)
   > Branch(Condition="Equal (Guid)"(A=@n6, B=Break Movie Scene Binding Proxy(MovieSceneBindingProxy=@n4)) -> (ReturnValue=~))
   > n7 = LoadAsset(AssetPath=BreakSoftObjectPath(InSoftObjectPath=@n5.Array Element_PathToAnimSequence))
   > n8 = Cast To AnimSequence(Object=@n7)
   > Add Unique(TargetArray:Array of Anim Sequence Object References=Get AnimSeq_Array(), NewItem:"Anim Sequence Object Reference (by ref)"=@n8)
   n9 = For Each Loop!(Exec=@n1.then_1, Array:Array of Anim Sequence Object References=Get AnimSeq_Array()) -> (LoopBody, Array Element:Anim Sequence Object Reference, Array Index, Completed)
   > Set Anim_Seq(Anim_Seq:Anim Sequence Object Reference=@n9) -> (Output_Get:Anim Sequence Object Reference)
   > n10 = Execute Python Script!(
    execute
    PythonScript="import unreal\r\n\r\n## Grab the Level Editor World\r\neditor_subsystem = unreal.get_editor_subsystem(unreal.UnrealEditorSubsystem)\r\nworld = editor_subsystem.get_editor_world()\r\n\r\nanim_seq_export_options = unreal.AnimSeqExportOption()\r\nanim_seq_export_options.export_transforms = True\r\nanim_seq_export_options.export_morph_targets = True\r\n"
   ) -> (then, ReturnValue, anim_seq_export_options:Anim Seq Export Option Object Reference) {outputs [anim_seq_export_options]}
   > Set Export_Options(Export_Options:Anim Seq Export Option Object Reference=@n10.anim_seq_export_options) -> (Output_Get:Anim Seq Export Option Object Reference)
   > ClearLinkedAnimSequences!(execute, InLevelSequence:Level Sequence Object Reference=Get CachedLevelSequence()) -> (then)
   > SaveLoadedAsset(AssetToSave=Get CachedLevelSequence(), bOnlyIfIsDirty=false)
   > LinkAnimSequence!(
    execute
    Sequence:Level Sequence Object Reference=Get CachedLevelSequence()
    AnimSequence:Anim Sequence Object Reference=Get Anim_Seq() -> (Anim_Seq:Anim Sequence Object Reference)
    ExportOptions:Anim Seq Export Option Object Reference=Get Export_Options() -> (Export_Options:Anim Seq Export Option Object Reference)
    Binding:"Movie Scene Binding Proxy Structure (by ref)"=Get target_binding() -> (target_binding:Movie Scene Binding Proxy Structure)
   ) -> (then, ReturnValue:bool)
   > SaveLoadedAssets(
    AssetsToSave=Make Array!("[0]":Object Reference=Get CachedLevelSequence(), "[1]":Object Reference=Get Anim_Seq()) -> (Array:Array of Object References)
    bOnlyIfIsDirty=false
   )
  ]}
 Not Current Warning {nodes [
   n1 = Not Current Warning!() -> (then, Type of Asset:String, Display Asset Name:String) {is=title K2Node_FunctionEntry}
   > Show Message Dialog(
    Title=Get Rig Replacement Tool Title()
    Message="To Text (String)"(InString=Append!(A=@n1, B=@n1.Display Asset Name, C=" is not the latest. Please get the latest for this asset and try again.") -> (ReturnValue))
    MessageType=Ok
    DefaultValue=No
    MessageCategory=Error
   )
   > Return Node!(execute, Could it be done:bool=false)
  ]}
 On_ScrollResize_Handle_MouseButtonDown {nodes [
   n1 = On_ScrollResize_Handle_MouseButtonDown!() -> (then, MyGeometry:Geometry Structure, MouseEvent:"Pointer Event Structure (by ref)") {is=title K2Node_FunctionEntry}
   > Set bDraggingScroll(bDraggingScroll:bool=true) -> (Output_Get:bool)
   > Set DragStartY(DragStartY:double=Get Screen Space Position!(Input:"Pointer Event Structure (by ref)"=@n1.MouseEvent) -> (ReturnValue_X:double, ReturnValue_Y:double) {is=title K2Node_CallFunction}.ReturnValue_Y) -> (Output_Get:double)
   > Set DragStartHeight(DragStartHeight:double=Get ScrollBoxHeight() -> (ScrollBoxHeight:double)) -> (Output_Get:double)
   > Return Node!(
    execute
    ReturnValue:Event Reply Structure=CaptureMouse!(
     Reply:"Event Reply Structure (by ref)"=Handled()
     CapturingWidget:Widget Object Reference=Get ScrollResize_Handle() -> (ScrollResize_Handle:Border Object Reference)
    ) -> (ReturnValue:Event Reply Structure)
   )
  ]}
 On_ScrollResize_Handle_MouseButtonUp {nodes [
   On_ScrollResize_Handle_MouseButtonUp!() -> (then, MyGeometry:Geometry Structure, MouseEvent:"Pointer Event Structure (by ref)") {is=title K2Node_FunctionEntry}
   > Set bDraggingScroll(bDraggingScroll=false)
   > Return Node(ReturnValue=ReleaseMouseCapture!(Reply:"Event Reply Structure (by ref)"=Handled()) -> (ReturnValue:Event Reply Structure))
  ]}
 On_ScrollResize_Handle_MouseMove {nodes [
   n1 = On_ScrollResize_Handle_MouseMove!() -> (then, MyGeometry:Geometry Structure, MouseEvent:"Pointer Event Structure (by ref)") {is=title K2Node_FunctionEntry}
   > n2 = Branch(Condition=Get bDraggingScroll() -> (bDraggingScroll:bool))
   > n3 = Set Current Y(Current Y:double=Get Screen Space Position(Input=@n1.MouseEvent).ReturnValue_Y) -> (Output_Get:double)
   > n4 = Set DeltaY(DeltaY:double=float - float(A=@n3, B=Get DragStartY() -> (DragStartY:double))) -> (Output_Get:double)
   > n5 = Set NewHeight(NewHeight:double=float + float(A=@n4, B=Get DragStartHeight() -> (DragStartHeight:double))) -> (Output_Get:double)
   > n6 = Set ClampedHeight(ClampedHeight:double="Clamp (Float)"(Value=@n5, Min=500, Max=1000)) -> (Output_Get:double)
   > n7 = Set ScrollBoxHeight(ScrollBoxHeight=@n6)
   > SetHeightOverride(self=Get SizeBox_Mapper(), InHeightOverride=@n7)
   > SetHeightOverride(
    self=Get Mapper_SizeBox!(self:EUW Mappings Widget Object Reference=Get MappingWidget()) -> (Mapper_SizeBox:Size Box Object Reference)
    InHeightOverride=@n7
   )
   > Return Node(ReturnValue=Handled())
   Return Node(execute=@n2.else, ReturnValue=Unhandled())
  ]}
 Paste Anim {nodes [
   Paste Anim() {is=title K2Node_FunctionEntry}
   > n1 = GetBindings(Sequence=Get CachedLevelSequence())
   > n2 = For Each Loop with Break!(Exec, Array:Array of Movie Scene Binding Proxy Structures=@n1, Break=@n9.then) -> (LoopBody, Array Element:Movie Scene Binding Proxy Structure, Array Index, Completed)
   n3 = For Each Loop with Break(Exec=@n2.Completed, Array=GetSelectedBindings(), Break=@n7.then)
   Execute Python Script!(
    execute=@n3.Completed
    PythonScript="import unreal\r\n\r\nsource_guid_str = source_binding.get_id().to_string()\r\ntarget_guid_str = target_binding.get_id().to_string()\r\n\r\nprint(f\"[AttachRewire] source_guid: {source_guid_str}\")\r\nprint(f\"[AttachRewire] target_guid: {target_guid_str}\")\r\n\r\n## Collect all valid GUIDs currently in the sequence\r\nall_binding_guids = set(b.get_id().to_string() for b in sequence.get_bindings())\r\n\r\nsections_to_rewire = []\r\n\r\nfor binding in sequence.get_bindings():\r\n    if binding.get_id().to_string() == target_guid_str:\r\n        continue\r\n    for track in binding.get_tracks():\r\n        if track.get_class().get_name() != \"MovieScene3DAttachTrack\":\r\n            continue\r\n        for section in track.get_sections():\r\n            current_guid_str = section.get_constraint_binding_id().get_editor_property('guid').to_string()\r\n            is_source = current_guid_str == source_guid_str\r\n            is_unresolved = current_guid_str not in all_binding_guids\r\n            if is_source or is_unresolved:\r\n                sections_to_rewire.append((binding.get_name(), section, current_guid_str))\r\n\r\nprint(f\"[AttachRewire] Found {len(sections_to_rewire)} attach section(s) to rewire\")\r\n\r\nwith unreal.ScopedEditorTransaction(\"Rewire Attach Tracks\") as trans:\r\n    for binding_name, section, old_guid in sections_to_rewire:\r\n        new_id = unreal.MovieSceneObjectBindingID()\r\n        new_id.set_editor_property('guid', target_binding.get_id())\r\n        section.modify()\r\n        section.set_editor_property('constraint_binding_id', new_id)\r\n        print(f\"[AttachRewire] Rewired: {binding_name} (was: {old_guid})\")\r\n\r\nunreal.LevelSequenceEditorBlueprintLibrary.refresh_current_level_sequence()\r\nprint(f\"[AttachRewire] Done.\")"
    sequence:Level Sequence Object Reference=Get CachedLevelSequence()
    source_binding:Movie Scene Binding Proxy Structure=Get source_binding()
    target_binding:Movie Scene Binding Proxy Structure=Get target_binding()
   ) -> (then, ReturnValue) {inputs [sequence, source_binding, target_binding]}
   > Execute Python Script!(
    execute
    PythonScript="import unreal\r\n\r\ncr_lib = unreal.ControlRigSequencerLibrary\r\n\r\n## Clean up layers that have no sections bound to anything at all\r\nlayers = cr_lib.get_anim_layers(add_if_does_not_exist=False)\r\nif layers:\r\n    for i in reversed(range(len(layers))):\r\n        layer = layers[i]\r\n        items = layer.get_editor_property('anim_layer_items')\r\n\r\n        has_any_section = False\r\n        for obj, item in items.items():\r\n            if item.section_items:\r\n                has_any_section = True\r\n                break\r\n\r\n        if not has_any_section:\r\n            unreal.log_warning(f\"Deleting empty/orphaned layer [{i}]: {layer.get_name()}\")\r\n            cr_lib.delete_anim_layer(i)"
   ) -> (then, ReturnValue)
   n4 = Get target_binding()
   Branch(execute=@n3.LoopBody, Condition=AND Boolean(A=IsValid(InBinding=@n3), B="Equal (MovieSceneBindingProxy)"(A=@n3, B=@n4) -> (ReturnValue=~)))
   > RefreshCurrentLevelSequence()
   > n5 = Return String Mapping!(execute, self:EUW Mappings Widget Object Reference=Get MappingWidget()) -> (then, Final Mapping:Map of Strings to Strings)
   n6 = Get LocalTransfer_Ckbx() -> (LocalTransfer_Ckbx:Editor Utility Check Box Object Reference)
   > n7 = Execute Python Script!(
    execute
    PythonScript="import unreal\r\nfrom AnimLibrary.core.libanim import importer\r\n\r\nanim_data = unreal.resolve_python_object_handle(export_data)\r\n\r\n## Flatten new exporter format to match what old Importer expects\r\nanim_data[\"sections\"] = anim_data.get(\"control_rig\", {}).get(\"sections\", [])\r\n\r\nmy_importer = importer.Importer(mappings=map, suppress_dialog=True, world_space=world_space)\r\nmy_importer.import_anim_data(anim_data=anim_data, asset_context={})"
    map:Map of Strings to Strings=@n5
    export_data:Python Object Handle Object Reference=Get export_data() -> (export_data:Python Object Handle Object Reference)
    world_space:bool=NOT Boolean(A=IsChecked(self=@n6))
   ) -> (then, ReturnValue) {inputs [map,export_data,world_space]}
   n8 = GetDisplayName!(execute:Exec=@n2.LoopBody, InBinding:"Movie Scene Binding Proxy Structure (by ref)"=@n2) -> (then:Exec, ReturnValue:Text)
   > Branch(Condition="Equal Exactly (String)"(A=Get target_sequence_actor() -> (target_sequence_actor:String), B="To String (Text)"(InText=@n8)))
   > SelectBindings(ObjectBindings=Make Array!("[0]":Movie Scene Binding Proxy Structure=@n2) -> (Array:Array of Movie Scene Binding Proxy Structures))
   > n9 = Set target_binding(target_binding:Movie Scene Binding Proxy Structure=@n2) -> (Output_Get:Movie Scene Binding Proxy Structure)
  ]}
 Populate Slots {nodes [
   Populate Slots() {is=title K2Node_FunctionEntry}
   > ClearChildren(self=Get LSB_Slot() -> (LSB_Slot:Named Slot Object Reference))
   > ClearChildren(self=Get MappingSlot() -> (MappingSlot:Named Slot Object Reference))
   > n1 = Create Widget(Class=/AnimInEdTools/LevelSequenceBrowser/EUW_LevelSequenceBrowser.EUW_LevelSequenceBrowser_C) -> (ReturnValue:EUW Level Sequence Browser Object Reference)
   > Set LevelSequenceBrowser(LevelSequenceBrowser:EUW Level Sequence Browser Soft Object Reference=Make Soft Reference!(Input:EUW Level Sequence Browser Object Reference=@n1) -> (Output:EUW Level Sequence Browser Soft Object Reference)) -> (Output_Get:EUW Level Sequence Browser Soft Object Reference)
   > SetSelectionMode!(
    execute
    self:List View Object Reference=Get ListView!(self:EUW Level Sequence Browser Object Reference=@n1) -> (ListView:Common List View Object Reference)
    SelectionMode:ESelectionMode Enum=Single
   ) -> (then)
   > n2 = Create Widget(Class=&/EUW_MappingsWidget.EUW_MappingsWidget_C) -> (ReturnValue:EUW Mappings Widget Object Reference)
   > Set MappingWidget(MappingWidget:EUW Mappings Widget Object Reference=@n2) -> (Output_Get:EUW Mappings Widget Object Reference)
   > AddChild(
    self=Get LSB_Slot()
    Content=Resolve Soft Reference(Input:EUW Level Sequence Browser Soft Object Reference=Get LevelSequenceBrowser()) -> (Output:EUW Level Sequence Browser Object Reference)
   )
   > AddChild(self=Get MappingSlot(), Content=Get MappingWidget())
   n3 = Resolve Soft Reference(Input=Get LevelSequenceBrowser())
   > Populate Source Rig(
    B=Get Selected_Asset!(self:EUW Level Sequence Browser Object Reference=@n3) -> (Selected_Asset:String)
    Project=Get Current Project!(self:EUW Level Sequence Browser Object Reference=@n3) -> (Current Project:BP Animation Browser Project Object Reference)
   )
  ]}
 Populate Source Rig {nodes [
   n1 = Populate Source Rig!() -> (then, B:String, Project:BP Animation Browser Project Object Reference) {is=title K2Node_FunctionEntry}
   > n2 = For Each Loop(Array:Array of Animatable Creature Object References=Get Animatable Creatures!(Project:BP Animation Browser Project Object Reference=@n1.Project, bSearchSubClasses:bool=false, __WorldContext:Object Reference) -> (Animatable Creatures:Array of Animatable Creature Object References) {is=title K2Node_CallFunction}) -> (Array Element:Animatable Creature Object Reference)
   n3 = Get MappingWidget()
   Return Node!(execute=[@n2.Completed, @n7.then], MappingWidget:EUW Mappings Widget Object Reference=@n3)
   > Set Source AC(Source AC:Animatable Creature Object Reference=@n2) -> (Output_Get:Animatable Creature Object Reference)
   > Branch(Condition="Equal Exactly (String)"(A=Get Asset!(self:Animatable Creature Object Reference=@n2) -> (Asset:String), B=@n1))
   n4 = Get Control Rig!(self:Animatable Creature Object Reference=@n2) -> (Control Rig:Control Rig Asset Soft Reference Structure)
   > n5 = LoadAsset(AssetPath=Split(
     SourceString="To String (Control Rig Soft Reference)"!(InRef:"Control Rig Asset Soft Reference Structure (by ref)"=@n4) -> (ReturnValue:String)
     InStr=.
     SearchDir=FromEnd
    ))
   > Is Valid(InputObject=@n5)
   > n6 = CastToControlRigBlueprint(Object=@n5)
   > n7 = Set SourceRig(SourceRig=@n6, self:EUW Mappings Widget Object Reference=@n3)
  ]}
 PopupDialog {nodes [
   PopupDialog() {is=title K2Node_FunctionEntry}
   n1 = GetDisplayName!(Object:Object Reference=Get SourceRig(self=Get MappingWidget())) -> (ReturnValue:String)
   n2 = GetDisplayName(Object=Get TargetRig(self=Get MappingWidget()))
   > Show Message Dialog(
    Title=Get Rig Replacement Tool Title()
    Message="To Text (String)"(InString=Append!(
      A="There are multiple "
      B=@n1
      C=" or "
      D:String=@n2
      E:String=" detected in the current levelSequence."
      F:String=" Please clean up the scene to have only one "
      G:String=@n1
      H:String=" and no "
      I:String=@n2
     ) -> (ReturnValue))
    MessageType=Ok
    DefaultValue=Ok
    MessageCategory=Error
   )
  ]}
 Query Anim Sequence {nodes [
   n1 = Query Anim Sequence!() -> (then, InFile:String, Anim Sequence:Anim Sequence Object Reference) {is=title K2Node_FunctionEntry}
   > n2 = QueryFileState!(execute, InFile:String=@n1, bSilent:bool=false, bUseSourceControlStateCache:bool=false) -> (
    then
    ReturnValue_Filename:String
    ReturnValue_bIsValid:bool=~
    ReturnValue_bIsUnknown:bool=~
    ReturnValue_bCanCheckIn:bool=~
    ReturnValue_bCanCheckOut:bool=~
    ReturnValue_bIsCheckedOut:bool=~
    ReturnValue_bIsCurrent:bool=~
    ReturnValue_bIsSourceControlled:bool=~
    ReturnValue_bIsAdded:bool=~
    ReturnValue_bIsDeleted:bool=~
    ReturnValue_bIsIgnored:bool=~
    ReturnValue_bCanEdit:bool=~
    ReturnValue_bCanDelete:bool=~
    ReturnValue_bIsModified:bool=~
    ReturnValue_bCanAdd:bool=~
    ReturnValue_bIsConflicted:bool=~
    ReturnValue_bCanRevert:bool=~
    ReturnValue_bIsCheckedOutOther:bool=~
    ReturnValue_CheckedOutOther:String
    ReturnValue_bIsCheckedOutInOtherBranch:bool=~
    ReturnValue_bIsModifiedInOtherBranch:bool=~
    ReturnValue_PreviousUser:String
   )
   > n3 = Branch(Condition=@n2.ReturnValue_bIsCurrent)
   > n4 = Branch(Condition=@n2.ReturnValue_bIsCheckedOutOther)
   > n5 = Checked out by another user warning!(execute, self, Checked out by:String=@n2.ReturnValue_CheckedOutOther, Dipslay aset name:String=GetDisplayName(Object=Get Anim Sequence())) -> (then, Was it done:bool) {is=title K2Node_CallFunction}
   > Return Node!(execute, Was it done:bool=@n5)
   Return Node(execute=@n4.else, Was it done=true)
   n6 = Not Current Warning!(execute=@n3.else, self, Type of Asset:String=Animation Sequence, Display Asset Name:String=GetDisplayName(Object=Get Anim Sequence())) -> (then, Could it be done:bool) {is=title K2Node_CallFunction}
   > Return Node(Was it done=@n6)
  ]}
 Query LevelSequence {nodes [
   n1 = Query LevelSequence!() -> (then, InFile:String, CurrentLS:Level Sequence Object Reference) {is=title K2Node_FunctionEntry}
   > n2 = QueryFileState(InFile=@n1, bSilent=false, bUseSourceControlStateCache=false) -> (
    ReturnValue_bIsValid=~
    ReturnValue_bIsUnknown=~
    ReturnValue_bCanCheckIn=~
    ReturnValue_bCanCheckOut=~
    ReturnValue_bIsCheckedOut=~
    ReturnValue_bIsCurrent=~
    ReturnValue_bIsSourceControlled=~
    ReturnValue_bIsAdded=~
    ReturnValue_bIsDeleted=~
    ReturnValue_bIsIgnored=~
    ReturnValue_bCanEdit=~
    ReturnValue_bCanDelete=~
    ReturnValue_bIsModified=~
    ReturnValue_bCanAdd=~
    ReturnValue_bIsConflicted=~
    ReturnValue_bCanRevert=~
    ReturnValue_bIsCheckedOutOther=~
    ReturnValue_bIsCheckedOutInOtherBranch=~
    ReturnValue_bIsModifiedInOtherBranch=~
   )
   > n3 = Branch(Condition=@n2.ReturnValue_bIsCurrent)
   > n4 = Branch(Condition=@n2.ReturnValue_bIsCheckedOutOther)
   > n5 = Checked out by another user warning(
    Checked out by=@n2.ReturnValue_CheckedOutOther
    Dipslay aset name=GetDisplayName(Object=Get CurrentLS() -> (CurrentLS:Level Sequence Object Reference))
   )
   > Return Node(Was it done=@n5)
   Return Node(execute=@n4.else, Was it done=true)
   n6 = Not Current Warning(execute=@n3.else, Type of Asset=Level Sequence, Display Asset Name=GetDisplayName(Object=Get CurrentLS()))
   > Return Node(Was it done=@n6)
  ]}
 Select Source Binding From Mults {nodes [
   Select Source Binding From Mults() {is=title K2Node_FunctionEntry}
   > Continue After Source Selected Binding(selected source binding=Get SourceBinding!(self:EUW Select Binding Object Reference=Get SelectedBinding()) -> (SourceBinding:Movie Scene Binding Proxy Structure))
  ]}
 Update Level Sequence Name {nodes [
   Update Level Sequence Name() {is=title K2Node_FunctionEntry}
   > n1 = GetFocusedLevelSequence()
   > n2 = Is Valid(InputObject=@n1)
   n3 = Get ActiveLevelSequence() -> (ActiveLevelSequence:Text Object Reference)
   Set Text(execute=@n2.Is Not Valid, Text="LOCTEXT(\"None\")", self=@n3)
   > Set!(
    execute
    ColorAndOpacity_SpecifiedColor:Linear Color Structure=(R=1,G=0.8778,B=0,A=1)
    ColorAndOpacity_ColorUseRule:ESlateColorStylingMode Enum
    self:Text Object Reference=@n3
   ) -> (then, Output_Get:Slate Color Structure)
   > Set Visibility(Visibility=Visible, self=Get SizeBox() -> (SizeBox:Size Box Object Reference))
   n4 = Get ActiveLevelSequence()
   Set Text(execute=@n2.Is Valid, Text="To Text (String)"(InString=GetDisplayName(Object=@n1)), self=@n4)
   > Set Visibility(Visibility=Collapsed, self=Get SizeBox())
   > Set(ColorAndOpacity_SpecifiedColor=(R=0,G=1,B=0,A=1), self=@n4)
  ]}
 Validation Of controlrigs in LevelSequence {nodes [
   Validation Of controlrigs in LevelSequence() {is=title K2Node_FunctionEntry}
   > Set There is a Source Rig(There is a Source Rig:bool=false) -> (Output_Get:bool)
   > Clear(TargetArray:Array of Control Rig Object References=Get TempTargetRigArray() -> (TempTargetRigArray:Array of Control Rig Object References))
   > n1 = GetFocusedLevelSequence()
   > n2 = GetControlRigs(LevelSequence=@n1)
   > n3 = For Each Loop!(Exec, Array:Array of Control Rig Sequencer Binding Proxy Structures=@n2) -> (
    LoopBody
    Array Element_Proxy:Movie Scene Binding Proxy Structure
    Array Element_ControlRig:Control Rig Object Reference
    Array Element_Track:Movie Scene Control Rig Parameter Track Object Reference
    Array Index
    Completed
   )
   n4 = Length(TargetArray:Array of Control Rig Object References=Get TempTargetRigArray())
   Return Node!(execute=@n3.Completed, Is Valid:bool=AND Boolean(A="Equal (Integer)"(A=@n4), B=Get There is a Source Rig() -> (There is a Source Rig:bool)))
   n5 = Get MappingWidget()
   n6 = Get TargetRig(self=@n5)
   n7 = GetDisplayName(Object=@n3.Array Element_ControlRig)
   > n8 = Branch(Condition="Equal Exactly (String)"(A=@n7, B=GetDisplayName(Object=@n6)))
   > Add(
    TargetArray:Array of Control Rig Object References=Get TempTargetRigArray()
    NewItem:"Control Rig Object Reference (by ref)"=@n3.Array Element_ControlRig
   ) {is=title K2Node_CallArrayFunction}
   n9 = Get SourceRig(self=@n5)
   Branch(execute=@n8.else, Condition="Equal Exactly (String)"(A=@n7, B=GetDisplayName(Object=@n9)))
   > Set There is a Source Rig(There is a Source Rig=true)
  ]}
 ValidationLevelSequence {nodes [
   ValidationLevelSequence() {is=title K2Node_FunctionEntry}
   > n1 = GetFocusedLevelSequence()
   > n2 = Set Selected Level Sequence(Selected Level Sequence:Level Sequence Object Reference=@n1) -> (Output_Get:Level Sequence Object Reference)
   > n3 = Is Valid(InputObject=@n2)
   Set bIsEnabled!(
    execute=@n3.Is Not Valid
    bIsEnabled:bool=false
    self:Widget Object Reference=Get Copy_Anim_Btn() -> (Copy_Anim_Btn:Editor Utility Button Object Reference)
   ) -> (then, Output_Get:bool)
   > Set bIsEnabled(bIsEnabled=false, self=Get MappingSlot())
   Set bIsEnabled(execute=@n3.Is Valid, bIsEnabled=true, self=Get Copy_Anim_Btn())
   > Set bIsEnabled(bIsEnabled=true, self=Get MappingSlot())
  ]}
]
variables [
 bDraggingScroll {type=boolean}
 CachedLevelSequence {type=LevelSequence, default=null}
 DragStartHeight {type=number}
 DragStartY {type=number}
 export_data {type=PythonObjectHandle, default=null}
 LevelSequenceBrowser {type=EUW_LevelSequenceBrowser_C, default=null}
 Mapping Row {type=EUW_MappingRow_C, default=null}
 MappingWidget {type=EUW_MappingsWidget_C, default=null}
 Rig Replacement Tool Title {type=string, default="Rig Replacement  Tool"}
 ScrollBoxHeight {type=number, default=600}
 Selected Level Sequence {type=LevelSequence, default=null}
 Selected LS Missing Modifier array {type="array<LevelSequence>", default []}
 SelectedBinding {type=EUW_SelectBinding_C, default=null}
 Source AC {type=AnimatableCreature_C, default=null}
 source_binding {type=MovieSceneBindingProxy, default {bindingId=#1, sequence=null}}
 Space {type=string, default=""}
 Start Time {type=integer}
 target_binding {type=MovieSceneBindingProxy, default {bindingId=#1, sequence=null}}
 target_sequence_actor {type=string, default=""}
]
widgets {
 info {rootWidgetClass=&CanvasPanel, widgetCount=31}
 widgets [
  CanvasPanel_21 {widgetClassPath=&CanvasPanel, widget=CanvasPanel_21}
  Overlay_173 {
   widgetClassPath=&Overlay
   parent=CanvasPanel_21
   widget=Overlay_173
   slot {class=CanvasPanelSlot, properties {zOrder=1, layoutData {alignment=<0 0>, anchors {maximum=<1 1>, minimum=<0 0>}, offsets=<ltrb 0 0 0 0>}}}
  }
  EUW_ContextMenu {
   widgetClassPath=/AnimInEdTools/ContextMenu/EUW_ContextMenu
   bIsVariable=true
   parent=Overlay_173
   widget=EUW_ContextMenu
   properties {visibility=Collapsed}
   slot {class=OverlaySlot}
  }
  ModalLayer {
   widgetClassPath=&NamedSlot
   bIsVariable=true
   parent=Overlay_173
   widget=ModalLayer
   properties {bExposeOnInstanceOnly=true}
   slot {class=OverlaySlot, properties {horizontalAlignment=HAlign_Fill, verticalAlignment=VAlign_Fill}}
  }
  Border_86 {
   widgetClassPath=&Border
   parent=CanvasPanel_21
   widget=Border_86
   properties {brushColor=<rgba 0.012 0.012 0.012 1>, padding=<ltrb 0 0 0 0>}
   slot {class=CanvasPanelSlot, properties {layoutData {alignment=<0 0>, anchors {maximum=<1 1>, minimum=<0 0>}, offsets=<ltrb 0 0 0 0>}}}
  }
  VerticalBox_1 {
   widgetClassPath=&VerticalBox
   parent=Border_86
   widget=VerticalBox_1
   slot {class=BorderSlot, properties {padding=<ltrb 0 0 0 0>}}
  }
  EditorUtilityScrollBox_2 {
   widgetClassPath=&EditorUtilityScrollBox
   parent=VerticalBox_1
   widget=EditorUtilityScrollBox_2
   slot {class=VerticalBoxSlot, properties {size {sizeRule=Fill, value=1}}}
  }
  VerticalBox_29 {widgetClassPath=&VerticalBox, parent=EditorUtilityScrollBox_2, widget=VerticalBox_29, slot {class=ScrollBoxSlot}}
  EditorUtilityScrollBox_0 {
   widgetClassPath=&EditorUtilityScrollBox
   parent=VerticalBox_29
   widget=EditorUtilityScrollBox_0
   slot {class=VerticalBoxSlot}
  }
  VerticalBox_0 {widgetClassPath=&VerticalBox, parent=EditorUtilityScrollBox_0, widget=VerticalBox_0, slot {class=ScrollBoxSlot}}
  SizeBox_0 {
   widgetClassPath=&SizeBox
   parent=VerticalBox_0
   widget=SizeBox_0
   properties {bOverride_HeightOverride=true, bOverride_WidthOverride=true, heightOverride=200, widthOverride=150}
   slot {class=VerticalBoxSlot}
  }
  LSB_Slot {widgetClassPath=&NamedSlot, bIsVariable=true, parent=SizeBox_0, widget=LSB_Slot, slot {class=SizeBoxSlot}}
  SizeBox_4 {
   widgetClassPath=&SizeBox
   parent=VerticalBox_0
   widget=SizeBox_4
   properties {bOverride_HeightOverride=true, heightOverride=28}
   slot {class=VerticalBoxSlot, properties {verticalAlignment=VAlign_Bottom, padding=<ltrb 10 10 0 5>, size {sizeRule=Fill, value=1}}}
  }
  HorizontalBox {widgetClassPath=&HorizontalBox, parent=SizeBox_4, widget=HorizontalBox, slot {class=SizeBoxSlot}}
  TextBlock {
   widgetClassPath=&TextBlock
   parent=HorizontalBox
   widget=TextBlock
   properties {
    text=Current Sequence:
    colorAndOpacity {colorUseRule=UseColor_Specified, specifiedColor=<rgba 1 1 1 0.8>}
    font {
     bForceMonospaced=false
     bMaterialIsStencil=false
     fontMaterial=null
     fontObject=&Roboto
     letterSpacing=0
     monospacedWidth=1
     size=10.5
     skewAmount=0
     typefaceFontName=Regular
     outlineSettings {
      bApplyOutlineToDropShadows=false
      bMiteredCorners=false
      bSeparateFillAlpha=false
      outlineMaterial=null
      outlineSize=0
      outlineColor=<rgba 0 0 0 1>
     }
    }
   }
   slot {class=HorizontalBoxSlot, properties {verticalAlignment=VAlign_Center, padding=<ltrb 0 0 5 0>}}
  }
  ActiveLevelSequence {
   widgetClassPath=&TextBlock
   bIsVariable=true
   parent=HorizontalBox
   widget=ActiveLevelSequence
   properties {font {
     bForceMonospaced=false
     bMaterialIsStencil=false
     fontMaterial=null
     fontObject=&Roboto
     letterSpacing=0
     monospacedWidth=1
     size=10.5
     skewAmount=0
     typefaceFontName=Regular
     outlineSettings {
      bApplyOutlineToDropShadows=false
      bMiteredCorners=false
      bSeparateFillAlpha=false
      outlineMaterial=null
      outlineSize=0
      outlineColor=<rgba 0 0 0 1>
     }
    }}
   slot {class=HorizontalBoxSlot, properties {verticalAlignment=VAlign_Center, padding=<ltrb 5 0 0 0>}}
  }
  SizeBox {
   widgetClassPath=&SizeBox
   bIsVariable=true
   parent=VerticalBox_0
   widget=SizeBox
   properties {bOverride_HeightOverride=true, heightOverride=28, visibility=Visible}
   slot {class=VerticalBoxSlot, properties {verticalAlignment=VAlign_Bottom, padding=<ltrb 10 0 0 0>, size {sizeRule=Fill, value=1}}}
  }
  DoubleClickLSBox {
   widgetClassPath=&HorizontalBox
   parent=SizeBox
   widget=DoubleClickLSBox
   properties {visibility=Visible}
   slot {class=SizeBoxSlot, properties {padding=<ltrb 0 0 0 10>}}
  }
  SelectALevelSequence {
   widgetClassPath=&TextBlock
   bIsVariable=true
   parent=DoubleClickLSBox
   widget=SelectALevelSequence
   properties {
    text=Double-click a sequence above.
    colorAndOpacity {colorUseRule=UseColor_Specified, specifiedColor=<rgba 1 1 1 0.8>}
    font {
     bForceMonospaced=false
     bMaterialIsStencil=false
     fontMaterial=null
     fontObject=&Roboto
     letterSpacing=0
     monospacedWidth=1
     size=10.5
     skewAmount=0
     typefaceFontName=Regular
     outlineSettings {
      bApplyOutlineToDropShadows=false
      bMiteredCorners=false
      bSeparateFillAlpha=false
      outlineMaterial=null
      outlineSize=0
      outlineColor=<rgba 0 0 0 1>
     }
    }
   }
   slot {class=HorizontalBoxSlot, properties {verticalAlignment=VAlign_Center, padding=<ltrb 0 0 5 0>}}
  }
  SizeBox_Mapper {
   widgetClassPath=&SizeBox
   bIsVariable=true
   parent=VerticalBox_0
   widget=SizeBox_Mapper
   properties {bOverride_HeightOverride=true, heightOverride=600}
   slot {class=VerticalBoxSlot}
  }
  EditorUtilityScrollBox_1 {widgetClassPath=&EditorUtilityScrollBox, parent=SizeBox_Mapper, widget=EditorUtilityScrollBox_1, slot {class=SizeBoxSlot}}
  MappingSlot {
   widgetClassPath=&NamedSlot
   bIsVariable=true
   parent=EditorUtilityScrollBox_1
   widget=MappingSlot
   slot {class=ScrollBoxSlot, properties {size {sizeRule=Fill, value=1}}}
  }
  ScrollResize_Handle {
   widgetClassPath=&Border
   bIsVariable=true
   parent=VerticalBox_29
   widget=ScrollResize_Handle
   properties {bOverride_Cursor=true, cursor=ResizeUpDown, brushColor=<rgba 0.1323 0.1323 0.1323 1>}
   slot {class=VerticalBoxSlot, properties {verticalAlignment=VAlign_Bottom, size {sizeRule=Fill, value=1}}}
  }
  HorizontalBox_3 {
   widgetClassPath=&HorizontalBox
   parent=VerticalBox_29
   widget=HorizontalBox_3
   properties {toolTipText=Transfers animation in local space instead of global. Use when control orientations match between rigs., visibility=Collapsed}
   slot {class=VerticalBoxSlot, properties {horizontalAlignment=HAlign_Center, verticalAlignment=VAlign_Center, padding=<ltrb 0 5 0 5>}}
  }
  LocalTransfer_Ckbx {
   widgetClassPath=&EditorUtilityCheckBox
   bIsVariable=true
   parent=HorizontalBox_3
   widget=LocalTransfer_Ckbx
   slot {class=HorizontalBoxSlot}
  }
  TextBlock_154 {
   widgetClassPath=&TextBlock
   parent=HorizontalBox_3
   widget=TextBlock_154
   properties {
    text=Transfer Animation in Local Space
    font {
     bForceMonospaced=false
     bMaterialIsStencil=false
     fontMaterial=null
     fontObject=&Roboto
     letterSpacing=0
     monospacedWidth=1
     size=12
     skewAmount=0
     typefaceFontName=Regular
     outlineSettings {
      bApplyOutlineToDropShadows=false
      bMiteredCorners=false
      bSeparateFillAlpha=false
      outlineMaterial=null
      outlineSize=0
      outlineColor=<rgba 0 0 0 1>
     }
    }
   }
   slot {class=HorizontalBoxSlot, properties {padding=<ltrb 5 5 5 5>}}
  }
  HorizontalBox_0 {
   widgetClassPath=&HorizontalBox
   parent=VerticalBox_1
   widget=HorizontalBox_0
   slot {class=VerticalBoxSlot, properties {horizontalAlignment=HAlign_Center, padding=<ltrb 10 10 10 10>}}
  }
  Copy_Anim_Btn {
   widgetClassPath=&EditorUtilityButton
   parent=HorizontalBox_0
   widget=Copy_Anim_Btn
   properties {toolTipText=Copy animations from Target Rig onto Source rig on the open LevelSequence}
   slot {class=HorizontalBoxSlot, properties {padding=<ltrb 5 0 5 0>}}
  }
  TextBlock_164 {
   widgetClassPath=&TextBlock
   parent=Copy_Anim_Btn
   widget=TextBlock_164
   properties {
    text=Apply Rig Replacement
    font {
     bForceMonospaced=false
     bMaterialIsStencil=false
     fontMaterial=null
     fontObject=&Roboto
     letterSpacing=0
     monospacedWidth=1
     size=13.5
     skewAmount=0
     typefaceFontName=Bold
     outlineSettings {
      bApplyOutlineToDropShadows=false
      bMiteredCorners=false
      bSeparateFillAlpha=false
      outlineMaterial=null
      outlineSize=0
      outlineColor=<rgba 0 0 0 1>
     }
    }
   }
   slot {class=ButtonSlot}
  }
  Button_? {
   widgetClassPath=&EditorUtilityButton
   parent=HorizontalBox_0
   widget=Button_?
   properties {
    toolTipText=Open Confluence Page
    widgetStyle {
     clickedSlateSound {resourceObject=null}
     disabled {
      bIsDynamicallyLoaded=false
      drawAs=RoundedBox
      imageType=NoImage
      mirroring=NoMirror
      resourceName=None
      resourceObject=null
      tiling=NoTile
      imageSize=<0 0>
      margin=<ltrb 0 0 0 0>
      outlineSettings {
       bUseBrushTransparency=false
       roundingType=FixedRadius
       width=1
       color {colorUseRule=UseColor_ColorTable, specifiedColor=<rgba 0 0 0 0>}
       cornerRadii=<4 4 4 4>
      }
      tintColor {colorUseRule=UseColor_Specified, specifiedColor=<rgba 0.0395 0.0395 0.0395 1>}
      uVRegion {bIsValid=false, max=<0 0>, min=<0 0>}
     }
     disabledForeground {colorUseRule=UseColor_ColorTable, specifiedColor=<rgba 0 0 0 0>}
     hovered {
      bIsDynamicallyLoaded=false
      drawAs=RoundedBox
      imageType=NoImage
      mirroring=NoMirror
      resourceName=None
      resourceObject=null
      tiling=NoTile
      imageSize=<32 32>
      margin=<ltrb 0 0 0 0>
      outlineSettings {
       bUseBrushTransparency=false
       roundingType=FixedRadius
       width=1
       color {colorUseRule=UseColor_ColorTable, specifiedColor=<rgba 0 0 0 0>}
       cornerRadii=<4 4 4 4>
      }
      tintColor {colorUseRule=UseColor_Specified, specifiedColor=<rgba 0.0953 0.0953 0.0953 1>}
      uVRegion {bIsValid=false, max=<0 0>, min=<0 0>}
     }
     hoveredForeground {colorUseRule=UseColor_ColorTable, specifiedColor=<rgba 0 0 0 0>}
     hoveredSlateSound {resourceObject=null}
     normal {
      bIsDynamicallyLoaded=false
      drawAs=RoundedBox
      imageType=NoImage
      mirroring=NoMirror
      resourceName=None
      resourceObject=null
      tiling=NoTile
      imageSize=<64 64>
      margin=<ltrb 0 0 0 0>
      outlineSettings {
       bUseBrushTransparency=false
       roundingType=FixedRadius
       width=1
       color {colorUseRule=UseColor_ColorTable, specifiedColor=<rgba 0 0 0 0>}
       cornerRadii=<4 4 4 4>
      }
      tintColor {colorUseRule=UseColor_Specified, specifiedColor=<rgba 0.0395 0.0395 0.0395 1>}
      uVRegion {bIsValid=false, max=<0 0>, min=<0 0>}
     }
     normalForeground {colorUseRule=UseColor_ColorTable, specifiedColor=<rgba 0 0 0 0>}
     normalPadding=<ltrb 8 2 6 2>
     pressed {
      bIsDynamicallyLoaded=false
      drawAs=RoundedBox
      imageType=NoImage
      mirroring=NoMirror
      resourceName=None
      resourceObject=null
      tiling=NoTile
      imageSize=<32 32>
      margin=<ltrb 0 0 0 0>
      outlineSettings {
       bUseBrushTransparency=false
       roundingType=FixedRadius
       width=1
       color {colorUseRule=UseColor_ColorTable, specifiedColor=<rgba 0 0 0 0>}
       cornerRadii=<4 4 4 4>
      }
      tintColor {colorUseRule=UseColor_Specified, specifiedColor=<rgba 0.0284 0.0284 0.0284 1>}
      uVRegion {bIsValid=false, max=<0 0>, min=<0 0>}
     }
     pressedForeground {colorUseRule=UseColor_ColorTable, specifiedColor=<rgba 0 0 0 0>}
     pressedPadding=<ltrb 8 2.5 6 2.5>
     pressedSlateSound {resourceObject=null}
    }
   }
   slot {
    class=HorizontalBoxSlot
    properties {horizontalAlignment=HAlign_Center, verticalAlignment=VAlign_Center, padding=<ltrb 10 0 10 0>, size {sizeRule=Fill, value=1}}
   }
  }
  TextBlock_14 {
   widgetClassPath=&TextBlock
   parent=Button_?
   widget=TextBlock_14
   properties {
    text=?
    colorAndOpacity {colorUseRule=UseColor_Specified, specifiedColor=<rgba 0 1 0 1>}
    font {
     bForceMonospaced=false
     bMaterialIsStencil=false
     fontMaterial=null
     fontObject=&Roboto
     letterSpacing=0
     monospacedWidth=1
     size=12
     skewAmount=0
     typefaceFontName=Bold
     outlineSettings {
      bApplyOutlineToDropShadows=false
      bMiteredCorners=false
      bSeparateFillAlpha=false
      outlineMaterial=null
      outlineSize=0
      outlineColor=<rgba 0 0 0 1>
     }
    }
   }
   slot {class=ButtonSlot, properties {horizontalAlignment=HAlign_Left, padding=<ltrb 0 0 0 0>}}
  }
 ]
}
